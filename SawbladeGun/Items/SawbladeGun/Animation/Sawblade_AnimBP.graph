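# Animation Blueprint for the Sawblade Gun weapon mesh, posing, spinning and hiding its three blades as ammo is spent.
# Event graph reads owner movement and weapon ammo, reload and charge state each update.
# An ammo-driven state machine feeds sprint or default montage slots, plus an additive charge intro.
# Bone modifiers then spin one selected blade, faster when charging or curve-driven during equip, and zero-scale hidden blades before final inertialization.
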AnimBlueprint /SawbladeGun/Items/SawbladeGun/Animation/Sawblade_AnimBP
target_skeleton=/SawbladeGun/Items/SawbladeGun/Meshes/Sawblade_Gun_Skeleton
graphs [
 AnimGraph {nodes [
   n1 = Blend Poses by bool(
    BlendPose_0=Slot 'DefaultSlotSprint'(Source=Use cached pose 'SawBlavedSavePose'()) {slotName=DefaultSlotSprint}
    BlendPose_1=Slot 'DefaultSlot'(Source=Use cached pose 'SawBlavedSavePose'())
    bActiveValue=Get IsSprinting() -> (IsSprinting:bool)
   ) {blendType=Linear, transitionType=Inertialization}
   AnimGraphNode_StateMachine_1 = ChargeIntroStateMachine() {is=title AnimGraphNode_StateMachine}
   n2 = Get SawBlade_RotationSpeed() -> (SawBlade_RotationSpeed:Rotator)
   n3 = "Transform (Modify) Bone - Bone: sawblade_01"!(
    ComponentPose=Local To Component(LocalPose=Apply Additive(Base=@n1, Additive=@AnimGraphNode_StateMachine_1) {alphaBoolBlend {blendOption=Linear}})
    bAlphaBoolEnabled:bool=Get SawBlade_01_AllowRotation() -> (SawBlade_01_AllowRotation:bool)
    Translation="0,0,0"
    Rotation=@n2
    Scale="1,1,1"
   ) -> (Pose) {alphaInputType=Bool, rotationMode=BMM_Additive, alphaBoolBlend {blendOption=Linear}, boneToModify {boneName=sawblade_01}}
   n4 = "Transform (Modify) Bone - Bone: sawblade_02"!(
    ComponentPose=@n3
    bAlphaBoolEnabled:bool=Get SawBlade_02_AllowRotation() -> (SawBlade_02_AllowRotation:bool)
    Translation="0,0,0"
    Rotation=@n2
    Scale="1,1,1"
   ) -> (Pose) {
    alphaInputType=Bool
    rotationMode=BMM_Additive
    alphaBoolBlend {blendInTime=0.1, blendOption=Linear, blendOutTime=0.1}
    boneToModify {boneName=sawblade_02}
   }
   n5 = "Transform (Modify) Bone - Bone: sawblade_03"!(
    ComponentPose=@n4
    bAlphaBoolEnabled:bool=Get SawBlade_03_AllowRotation() -> (SawBlade_03_AllowRotation:bool)
    Translation="0,0,0"
    Rotation=@n2
    Scale="1,1,1"
   ) -> (Pose) {
    alphaInputType=Bool
    rotationMode=BMM_Additive
    alphaBoolBlend {blendInTime=0.1, blendOption=Linear, blendOutTime=0.1}
    boneToModify {boneName=sawblade_03}
   }
   n6 = "Transform (Modify) Bone - Bone: sawblade_01"(
    ComponentPose=@n5
    bAlphaBoolEnabled=Get SawBlade_01_Scale() -> (SawBlade_01_Scale:bool)
    Translation="0,0,0"
    Rotation="0,0,0"
    Scale="0,0,0"
   ) {
    alphaInputType=Bool
    rotationMode=BMM_Additive
    scaleMode=BMM_Replace
    alphaBoolBlend {blendOption=Linear}
    boneToModify {boneName=sawblade_01}
    scale=<0 0 0>
   }
   n7 = "Transform (Modify) Bone - Bone: sawblade_02"(
    ComponentPose=@n6
    bAlphaBoolEnabled=Get SawBlade_02_Scale() -> (SawBlade_02_Scale:bool)
    Translation="0,0,0"
    Rotation="0,0,0"
    Scale="0,0,0"
   ) {
    alphaInputType=Bool
    rotationMode=BMM_Additive
    scaleMode=BMM_Replace
    alphaBoolBlend {blendOption=Linear}
    boneToModify {boneName=sawblade_02}
    scale=<0 0 0>
   }
   n8 = "Transform (Modify) Bone - Bone: sawblade_03"(
    ComponentPose=@n7
    bAlphaBoolEnabled=Get SawBlade_03_Scale() -> (SawBlade_03_Scale:bool)
    Translation="0,0,0"
    Rotation="0,0,0"
    Scale="0,0,0"
   ) {
    alphaInputType=Bool
    rotationMode=BMM_Additive
    scaleMode=BMM_Replace
    alphaBoolBlend {blendOption=Linear}
    boneToModify {boneName=sawblade_03}
    scale=<0 0 0>
   }
   Output Pose(Result=Inertialization!(Source:Pose Link Structure=Component To Local(ComponentPose=@n8)) -> (Pose:Pose Link Structure) {is=title AnimGraphNode_Inertialization}) {layerGroup=None}
   AnimGraphNode_StateMachine_0 = SawBlade() {is=title AnimGraphNode_StateMachine, bAllowConduitEntryStates=true}
   Save cached pose 'SawBlavedSavePose'(Pose=@AnimGraphNode_StateMachine_0) {cacheName=SawBlavedSavePose}
  ]}
 SawBlade {
  entry=Conduit
  path=AnimGraph.AnimGraphNode_StateMachine_0.SawBlade
  ids=id
  nodes [
   AnimStateConduitNode_0 = Conduit()
   AnimStateNode_3 = 3_blade_Idle() {is=title AnimStateNode}
   AnimStateNode_4 = 3_Blade_Fire() {is=title AnimStateNode}
   AnimStateNode_5 = 2_Blade_Idle() {is=title AnimStateNode}
   AnimStateNode_6 = 2_Blade_Fire() {is=title AnimStateNode}
   AnimStateNode_7 = I_Blade_Fire() {is=title AnimStateNode}
   AnimStateNode_8 = 0_Blade_idle() {is=title AnimStateNode}
   AnimStateTransitionNode_0 = AnimStateTransitionNode() {crossfadeDuration=0, logicType=TLT_Inertialization}
   AnimStateTransitionNode_1 = AnimStateTransitionNode() {crossfadeDuration=0.1, logicType=TLT_Inertialization, priorityOrder=100}
   AnimStateTransitionNode_10 = AnimStateTransitionNode() {crossfadeDuration=0, logicType=TLT_Inertialization}
   AnimStateTransitionNode_2 = AnimStateTransitionNode() {crossfadeDuration=0, logicType=TLT_Inertialization, priorityOrder=2}
   AnimStateTransitionNode_3 = AnimStateTransitionNode() {crossfadeDuration=0, logicType=TLT_Inertialization}
   AnimStateTransitionNode_7 = AnimStateTransitionNode() {bAutomaticRuleBasedOnSequencePlayerInState=true, crossfadeDuration=0.1, logicType=TLT_Inertialization}
   AnimStateTransitionNode_8 = AnimStateTransitionNode() {crossfadeDuration=0, logicType=TLT_Inertialization}
   AnimStateTransitionNode_9 = AnimStateTransitionNode() {bAutomaticRuleBasedOnSequencePlayerInState=true, crossfadeDuration=0.1, logicType=TLT_Inertialization}
  ]
  links [
   {
    dst=I_Blade_Fire
    rule=AnimGraph.AnimGraphNode_StateMachine_0.SawBlade.AnimStateTransitionNode_9.AnimationTransitionGraph_0
    src=2_Blade_Fire
    via=AnimStateTransitionNode_9
   }
   {
    dst=2_Blade_Fire
    rule=AnimGraph.AnimGraphNode_StateMachine_0.SawBlade.AnimStateTransitionNode_8.AnimationTransitionGraph_0
    src=2_Blade_Idle
    via=AnimStateTransitionNode_8
   }
   {
    dst=2_Blade_Idle
    rule=AnimGraph.AnimGraphNode_StateMachine_0.SawBlade.AnimStateTransitionNode_7.AnimationTransitionGraph_0
    src=3_Blade_Fire
    via=AnimStateTransitionNode_7
   }
   {
    dst=3_Blade_Fire
    rule=AnimGraph.AnimGraphNode_StateMachine_0.SawBlade.AnimStateTransitionNode_0.AnimationTransitionGraph_0
    src=3_blade_Idle
    via=AnimStateTransitionNode_0
   }
   {
    dst=2_Blade_Idle
    rule=AnimGraph.AnimGraphNode_StateMachine_0.SawBlade.AnimStateTransitionNode_2.AnimationTransitionGraph_1
    src=Conduit
    via=AnimStateTransitionNode_2
   }
   {
    dst=3_blade_Idle
    rule=AnimGraph.AnimGraphNode_StateMachine_0.SawBlade.AnimStateTransitionNode_1.AnimationTransitionGraph_1
    src=Conduit
    via=AnimStateTransitionNode_1
   }
   {
    dst=I_Blade_Fire
    rule=AnimGraph.AnimGraphNode_StateMachine_0.SawBlade.AnimStateTransitionNode_3.AnimationTransitionGraph_1
    src=Conduit
    via=AnimStateTransitionNode_3
   }
   {
    dst=0_Blade_idle
    rule=AnimGraph.AnimGraphNode_StateMachine_0.SawBlade.AnimStateTransitionNode_10.AnimationTransitionGraph_0
    src=I_Blade_Fire
    via=AnimStateTransitionNode_10
   }
  ]
 }
 Conduit {path=AnimGraph.AnimGraphNode_StateMachine_0.SawBlade.AnimStateConduitNode_0.Conduit, nodes [Result(bCanEnterTransition=true)]}
 3_blade_Idle {
  path=AnimGraph.AnimGraphNode_StateMachine_0.SawBlade.AnimStateNode_3.3_blade_Idle
  nodes [Output Animation Pose(Result=Sequence Player 'SawBlade_Idle_WEP'() {sequence=&/SawBlade_Idle_WEP})]
 }
 3_Blade_Fire {
  path=AnimGraph.AnimGraphNode_StateMachine_0.SawBlade.AnimStateNode_4.3_Blade_Fire
  nodes [Output Animation Pose(Result=Sequence Player 'SawBlade_Fire_3bade_WEP'() {sequence=&/SawBlade_Fire_3bade_WEP})]
 }
 2_Blade_Idle {
  path=AnimGraph.AnimGraphNode_StateMachine_0.SawBlade.AnimStateNode_5.2_Blade_Idle
  nodes [Output Animation Pose(Result=Sequence Player 'SawBlade_Fire_2Blade_POSE_WEP'() {sequence=&/SawBlade_Fire_2Blade_POSE_WEP})]
 }
 2_Blade_Fire {
  path=AnimGraph.AnimGraphNode_StateMachine_0.SawBlade.AnimStateNode_6.2_Blade_Fire
  nodes [Output Animation Pose(Result=Sequence Player 'SawBlade_Fire_2Blade_WEP'() {sequence=&/SawBlade_Fire_2Blade_WEP})]
 }
 I_Blade_Fire {
  path=AnimGraph.AnimGraphNode_StateMachine_0.SawBlade.AnimStateNode_7.I_Blade_Fire
  nodes [Output Animation Pose(Result=Sequence Player 'SawBlade_Fire_1Blade_POSE_WEP'() {sequence=&/SawBlade_Fire_1Blade_POSE_WEP})]
 }
 0_Blade_idle {
  path=AnimGraph.AnimGraphNode_StateMachine_0.SawBlade.AnimStateNode_8.0_Blade_idle
  nodes [Output Animation Pose(Result=Sequence Player 'SawBlade_Reload_WEP'() {playRate=0, sequence=&/SawBlade_Reload_WEP})]
 }
 AnimationTransitionGraph_0 {
  path=AnimGraph.AnimGraphNode_StateMachine_0.SawBlade.AnimStateTransitionNode_0.AnimationTransitionGraph_0
  nodes [Result(bCanEnterTransition="Equal (Integer)"(A=Get AmmoCount() -> (AmmoCount:int=0), B=2))]
 }
 AnimationTransitionGraph_1 {
  path=AnimGraph.AnimGraphNode_StateMachine_0.SawBlade.AnimStateTransitionNode_1.AnimationTransitionGraph_1
  nodes [Result(bCanEnterTransition=true)]
 }
 AnimationTransitionGraph_0 {
  path=AnimGraph.AnimGraphNode_StateMachine_0.SawBlade.AnimStateTransitionNode_10.AnimationTransitionGraph_0
  nodes [Result(bCanEnterTransition="Equal (Integer)"(A=Get AmmoCount() -> (AmmoCount=0)))]
 }
 AnimationTransitionGraph_1 {
  path=AnimGraph.AnimGraphNode_StateMachine_0.SawBlade.AnimStateTransitionNode_2.AnimationTransitionGraph_1
  nodes [Result(bCanEnterTransition="Equal (Integer)"(A=Get AmmoCount() -> (AmmoCount=0), B=2))]
 }
 AnimationTransitionGraph_1 {
  path=AnimGraph.AnimGraphNode_StateMachine_0.SawBlade.AnimStateTransitionNode_3.AnimationTransitionGraph_1
  nodes [Result(bCanEnterTransition="Equal (Integer)"(A=Get AmmoCount() -> (AmmoCount=0), B=1))]
 }
 AnimationTransitionGraph_0 {path=AnimGraph.AnimGraphNode_StateMachine_0.SawBlade.AnimStateTransitionNode_7.AnimationTransitionGraph_0, nodes [Result()]}
 AnimationTransitionGraph_0 {
  path=AnimGraph.AnimGraphNode_StateMachine_0.SawBlade.AnimStateTransitionNode_8.AnimationTransitionGraph_0
  nodes [Result(bCanEnterTransition="Equal (Integer)"(A=Get AmmoCount() -> (AmmoCount=0), B=1))]
 }
 AnimationTransitionGraph_0 {path=AnimGraph.AnimGraphNode_StateMachine_0.SawBlade.AnimStateTransitionNode_9.AnimationTransitionGraph_0, nodes [Result()]}
 ChargeIntroStateMachine {
  entry=Idenity State
  path=AnimGraph.AnimGraphNode_StateMachine_1.ChargeIntroStateMachine
  ids=id
  nodes [
   AnimStateNode_0 = Idenity State() {is=title AnimStateNode}
   AnimStateNode_1 = Charge_Intro() {is=title AnimStateNode}
   AnimStateTransitionNode_0 = AnimStateTransitionNode() {logicType=TLT_Inertialization}
   AnimStateTransitionNode_1 = AnimStateTransitionNode() {logicType=TLT_Inertialization}
  ]
  links [
   {
    dst=Idenity State
    rule=AnimGraph.AnimGraphNode_StateMachine_1.ChargeIntroStateMachine.AnimStateTransitionNode_1.Transition
    src=Charge_Intro
    via=AnimStateTransitionNode_1
   }
   {
    dst=Charge_Intro
    rule=AnimGraph.AnimGraphNode_StateMachine_1.ChargeIntroStateMachine.AnimStateTransitionNode_0.Transition
    src=Idenity State
    via=AnimStateTransitionNode_0
   }
  ]
 }
 Idenity State {
  path=AnimGraph.AnimGraphNode_StateMachine_1.ChargeIntroStateMachine.AnimStateNode_0.Idenity State
  nodes [Output Animation Pose(Result=Additive Identity Pose())]
 }
 Charge_Intro {
  path=AnimGraph.AnimGraphNode_StateMachine_1.ChargeIntroStateMachine.AnimStateNode_1.Charge_Intro
  nodes [Output Animation Pose(Result="Sequence Player (additive) 'SawBlade_Fire_Charge_Intro_WEP'"() {bLoopAnimation=false, sequence=&/SawBlade_Fire_Charge_Intro_WEP})]
 }
 Transition {
  path=AnimGraph.AnimGraphNode_StateMachine_1.ChargeIntroStateMachine.AnimStateTransitionNode_0.Transition
  nodes [Result(bCanEnterTransition=Get Ischarging() -> (Ischarging:bool))]
 }
 Transition {
  path=AnimGraph.AnimGraphNode_StateMachine_1.ChargeIntroStateMachine.AnimStateTransitionNode_1.Transition
  nodes [Result(bCanEnterTransition=NOT Boolean(A=Get Ischarging()))]
 }
 EventGraph {nodes [
   Event BlueprintUpdateAnimation()
   > n1 = Cast To PlayerPawn_Athena(Object=TryGetPawnOwner())
   > n2 = Cast To FortPlayerAnimInstance(Object=GetAnimInstance(self=Get Mesh(self=@n1)))
   > n3 = Cast To B_Ranged_Generic!(execute, Object=GetOwningActor()) -> (then, CastFailed, AsB Ranged Generic:B Ranged Generic Object Reference) {is=title K2Node_DynamicCast}
   n4 = Get bIsSliding!(self:Fort Anim Instance Object Reference=@n2) -> (bIsSliding)
   n5 = Get bIsJumping!(self:Fort Anim Instance Object Reference=@n2) -> (bIsJumping)
   n6 = Get bIsFalling!(self:Fort Anim Instance Object Reference=@n2) -> (bIsFalling)
   n7 = Get Speed2D!(self:Fort Player Anim Instance Object Reference=@n2) -> (Speed2D:float)
   > Set IsSprinting(IsSprinting:bool=AND Boolean!(A="float >= float"(A=@n7, B=400), B=NOT Boolean(A=@n5), C:bool=NOT Boolean(A=@n6), D:bool=NOT Boolean(A=@n4)) -> (ReturnValue)) -> (Output_Get:bool)
   > Set AmmoCount(AmmoCount:int=GetLocalMagazineAmmoCount!(self:Fort Weapon Object Reference=@n3) -> (ReturnValue:int=0)) -> (Output_Get:int=0)
   > Set IsReloading(IsReloading:bool=IsReloading!(self:Fort Weapon Object Reference=@n3) -> (ReturnValue:bool)) -> (Output_Get:bool)
   > Set Ischarging(Ischarging:bool=GetIsChargingWeapon!(self:Fort Weapon Object Reference=@n3) -> (ReturnValue:bool)) -> (Output_Get:bool)
   > n8 = Sequence!(execute) -> (then_0, then_1, then_2:Exec, then_3:Exec, then_4:Exec, then_5:Exec)
   > n9 = Branch(Condition=OR Boolean(
     A=Montage_IsPlaying(Montage=&/SawBlade_Equip_Sprint_WEP_M)
     B=Montage_IsPlaying(Montage=&/SawBlade_Equip_WEP_M)
    ))
   > Set SawBladePlayEquip(SawBladePlayEquip:bool=true) -> (Output_Get:bool)
   Set SawBladePlayEquip(execute=@n9.else, SawBladePlayEquip=false)
   n10 = Branch(execute=@n8.then_1, Condition=Get Ischarging())
   > Set RotationMultiplier(RotationMultiplier:double=10) -> (Output_Get:double)
   n11 = Branch(execute=@n10.else, Condition=Get SawBladePlayEquip() -> (SawBladePlayEquip:bool))
   > Set BladeEquipRotation(BladeEquipRotation:double=GetCurveValue(CurveName=BladeRotation)) -> (Output_Get:double)
   Set RotationMultiplier(execute=@n11.else, RotationMultiplier=0.75)
   n12 = Increment Float!(" ":Exec=@n8.then_2, Value:"Float (double-precision) (by ref)"=Get BladeRotation() -> (BladeRotation:double)) -> ("  ":Exec, Result:double=~) {is=title K2Node_MacroInstance}
   > n13 = Set BladeRotation(BladeRotation:double=@n12) -> (Output_Get:double)
   > n14 = Branch(Condition=Get SawBladePlayEquip())
   > Set SawBlade_RotationSpeed(SawBlade_RotationSpeed:Rotator=MakeRotator(Roll=Get BladeEquipRotation() -> (BladeEquipRotation:double))) -> (Output_Get:Rotator)
   n15 = Get RotationMultiplier() -> (RotationMultiplier:double)
   Set SawBlade_RotationSpeed(execute=@n14.else, SawBlade_RotationSpeed=MakeRotator(Roll=float * float(A=@n13, B=@n15) {is=title K2Node_PromotableOperator}))
   n16 = Get AmmoCount() -> (AmmoCount=0)
   n17 = Branch(
    execute=@n8.then_4
    Condition=OR Boolean!(
     A=Get SawBladePlayEquip()
     B=Get IsReloading() -> (IsReloading:bool)
     C:bool="Equal (Integer)"(A=@n16)
     D:bool="Equal (Integer)"(A=@n16, B=3)
    ) -> (ReturnValue)
   )
   > Set SawBlade_01_AllowRotation(SawBlade_01_AllowRotation:bool=true) -> (Output_Get:bool)
   > Set SawBlade_02_AllowRotation(SawBlade_02_AllowRotation:bool=false) -> (Output_Get:bool)
   > Set SawBlade_03_AllowRotation(SawBlade_03_AllowRotation:bool=false) -> (Output_Get:bool)
   n18 = Branch(execute=@n17.else, Condition="Equal (Integer)"(A=Get AmmoCount() -> (AmmoCount=0), B=2))
   > Set SawBlade_01_AllowRotation(SawBlade_01_AllowRotation=false)
   > Set SawBlade_02_AllowRotation(SawBlade_02_AllowRotation=true)
   > Set SawBlade_03_AllowRotation(SawBlade_03_AllowRotation=false)
   Set SawBlade_01_AllowRotation(execute=@n18.else, SawBlade_01_AllowRotation=false)
   > Set SawBlade_02_AllowRotation(SawBlade_02_AllowRotation=false)
   > Set SawBlade_03_AllowRotation(SawBlade_03_AllowRotation=true)
   n19 = Get AmmoCount() -> (AmmoCount=0)
   n20 = Branch(execute=@n8.then_5, Condition=OR Boolean!(A=Get IsReloading(), B=Get SawBladePlayEquip(), C="Equal (Integer)"(A=@n19, B=3)) -> (ReturnValue))
   n21 = Get AmmoCount() -> (AmmoCount=0)
   > n22 = Branch(Condition=OR Boolean!(A=Get IsReloading(), B="Equal (Integer)"(A=@n21, B=3)) -> (ReturnValue))
   > Set SawBlade_01_Scale(SawBlade_01_Scale:bool=false) -> (Output_Get:bool)
   > Set SawBlade_02_Scale(SawBlade_02_Scale:bool=false) -> (Output_Get:bool)
   > Set SawBlade_03_Scale(SawBlade_03_Scale:bool=false) -> (Output_Get:bool)
   n23 = Branch(execute=@n22.else, Condition="Equal (Integer)"(A=Get AmmoCount() -> (AmmoCount=0), B=2))
   > Set SawBlade_01_Scale(SawBlade_01_Scale=false)
   > Set SawBlade_02_Scale(SawBlade_02_Scale=false)
   > Set SawBlade_03_Scale(SawBlade_03_Scale=true)
   n24 = Branch(execute=@n23.else, Condition="Equal (Integer)"(A=Get AmmoCount() -> (AmmoCount=0), B=1))
   > Set SawBlade_01_Scale(SawBlade_01_Scale=false)
   > Set SawBlade_02_Scale(SawBlade_02_Scale=true)
   > Set SawBlade_03_Scale(SawBlade_03_Scale=true)
   Set SawBlade_01_Scale(execute=@n24.else, SawBlade_01_Scale=true)
   > Set SawBlade_02_Scale(SawBlade_02_Scale=true)
   > Set SawBlade_03_Scale(SawBlade_03_Scale=true)
   n25 = Branch(execute=@n20.else, Condition="Equal (Integer)"(A=Get AmmoCount() -> (AmmoCount=0), B=2))
   > Set SawBlade_01_Scale(SawBlade_01_Scale=true)
   > Set SawBlade_02_Scale(SawBlade_02_Scale=false)
   > Set SawBlade_03_Scale(SawBlade_03_Scale=false)
   Set SawBlade_01_Scale(execute=@n25.else, SawBlade_01_Scale=true)
   > Set SawBlade_02_Scale(SawBlade_02_Scale=true)
   > Set SawBlade_03_Scale(SawBlade_03_Scale=false)
  ]}
]
variables [
 AmmoCount {type=integer}
 BladeEquipRotation {type=number, default=10}
 BladeRotation {type=number, default=10}
 Fort Ranged Weapon Recoil {type=FortRangedWeaponRecoil, default {bulletNum=0, maxHorizontal=0, maxVertical=0, minHorizontal=0, minVertical=0, recoveryTime=0}}
 Ischarging {type=boolean}
 IsReloading {type=boolean}
 IsSprinting {type=boolean}
 RotationMultiplier {type=number}
 SawBlade_01_AllowRotation {type=boolean}
 SawBlade_01_Scale {type=boolean}
 SawBlade_02_AllowRotation {type=boolean}
 SawBlade_02_Scale {type=boolean}
 SawBlade_03_AllowRotation {type=boolean}
 SawBlade_03_Scale {type=boolean}
 SawBlade_RotationSpeed {type=Rotator}
 SawBladePlayEquip {type=boolean}
]
