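; Fortnite weapon Blueprint for the Sawblade Gun, a charge-up launcher that also deals melee damage while charging.
; Charging drives a local trajectory spline preview, staged audio, VFX and force feedback, and the projectile's launch speed.
; For melee, the owning client sends muzzle position on a timer; the server sphere-traces, applies damage and knockback.
; It also manages surface-specific damage audio, holster-aware idle sound, and an event removing the weapon when ammo runs out.
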(event EventOnPlayWeaponFireFX
  (params (bPersistentFire :type Boolean) (bSecondaryFire :type Boolean))
  ;[0.4132 0.02567 1] EventOnPlayWeaponFireFX
  ; Call Weapon Fire FX for Niagra System
  (|CallParentFunction bPersistentFire bSecondaryFire)
  (Components|Activation|Activate (Variables|Reloading|GetNS_Sawbladegun_MuzzleFX) false)
  (bind _playerpawn (Variables|Cached|GetPlayerPawn)
    (:then
      (GameplayCue|ExecuteGameplayCue _playerpawn (Variables|Firing|GetGCTag_FiringFF) (Class|AthenaFunctionLibrary|MakeSimpleFortEffectContext (Variables|Actor|GetInstigator) self))
      (if (<= (FortWeapon|GetMagazineAmmoCount) 1)
        (Audio|SpawnSoundAttached (Variables|Default|GetSFX_LastShotInClip) (Variables|Visual|GetWeaponMesh))))
    (:else)))

(event EventBeginPlay
  ;[0.1083 1 0.0636] EventBeginPlay
  ; Bind events and setup
  (|CallParentFunction)
  ;;[0.1083 1 0.0636] Bind Events
  (EventDispatchers|BindEventtoOnLocalAmmoChanged (EventDispatchers|CreateEvent :!event Custom|OnLocalAmmoChange)))

(event EventOnPlayReloadFX
  (params (ReloadStage :type EFortReloadFXState%20Enum :default "ReloadStart"))
  ;[0.4132 0.02567 1] EventOnPlayReloadFX
  ; Call Weapon Reload FX for Niagra System
  (|CallParentFunction ReloadStage)
  (switch Utilities|FlowControl|Switch|SwitchonEFortReloadFXState ReloadStage
    (:ReloadStart
      (Audio|Components|Audio|FadeOut (Variables|Default|GetSFXc_Passive) 0.1 0.0 "Linear" :!class /Script/Engine.AudioComponent))
    (:ReloadEnd
      (CallFunction|UnholsteredEvent))))

(event FortWeapon|EventOnUnEquippedAndDetached
  ;[0.04769 0.06954 0.4219 0.391 42] CHARGE UP AND TRAJECTORY
  ;;[0.4132 0.02567 1 1 32] VFX/SFX
  (Components|Activation|Deactivate (Variables|Default|GetNS_SawBlade_Chargeup))
  (Variables|Default|SetbHolstered true)
  (EventDispatchers|UnbindEventfromOnHolsteredEvent (Variables|Default|GetWeaponOwnerReference) (EventDispatchers|CreateEvent :!event Custom|HolsteredEvent))
  (EventDispatchers|UnbindEventfromOnUnholsteredEvent (Variables|Default|GetWeaponOwnerReference) (EventDispatchers|CreateEvent :!event Custom|UnholsteredEvent))
  (EventDispatchers|UnbindEventfromOnPawnStartedEmote (Variables|Default|GetWeaponOwnerReference) (EventDispatchers|CreateEvent :!event Custom|EmoteStarted))
  (EventDispatchers|UnbindEventfromOnPawnStoppedEmote (Variables|Default|GetWeaponOwnerReference) (EventDispatchers|CreateEvent :!event Custom|EmoteStopped))
  (Default|UnbindEventfromOnAimDownSightsChanged (EventDispatchers|CreateEvent :!event Custom|AimDownSightsChangedEvent)))

(event EventOnStartCharge
  ;[0.04769 0.06954 0.4219 0.391 42] CHARGE UP AND TRAJECTORY
  ;;[0.1083 1 0.0636] EventOnStartCharge
  ;; Call Parent, initialize melee damage, and if a trajectory is not yet initialized, then initialize a new trajectory
  (|CallParentFunction)
  (CallFunction|StartMelee)
  (Variables|Trajectory|GetSawBladeCustomTrajectory
    (:then
      _)
    (:else
      (CallFunction|InitializeProjectileTrajectory))) ; check if already has a projectile trajectory
  (Networking|SwitchHasAuthority
    (:Authority
      (bind _playerpawn (Variables|Cached|GetPlayerPawn)
        (:then
          (GameplayCue|AddGameplayCueOnActor(Looping) _playerpawn (Variables|ChargeUp|GetGCTag_ChargeUp) (Ability|GameplayCue|MakeGameplayCueParameters 0.0)))
        (:else)))
    (:Remote
      (ForceFeedback|PlayLocalForceFeedback (Variables|ForceFeedback|GetStartChargeForceFeedback) (Variables|ForceFeedback|GetStartChargeForceFeedbackTag) true))))

(event FortWeapon|EventOnEquippedAndAttached
  ;[0.04769 0.06954 0.4219 0.391 42] CHARGE UP AND TRAJECTORY
  ;;[0.4132 0.02567 1 1 32] VFX/SFX
  (Components|Activation|Activate (Variables|Default|GetNS_SawBlade_Chargeup) false)
  (Variables|Melee|SetRepSetMeleeVFX false)
  (bind _asplayer_pawn_athena_generic_parent (Utilities|Casting|CastToPlayerPawn_Athena_Generic_Parent (Game|GetInstigator)))
  (Variables|Default|SetWeaponOwnerReference _asplayer_pawn_athena_generic_parent)
  (Variables|Default|SetbHolstered false)
  (EventDispatchers|BindEventtoOnHolsteredEvent (Variables|Default|GetWeaponOwnerReference) (EventDispatchers|CreateEvent :!event Custom|HolsteredEvent))
  (EventDispatchers|BindEventtoOnUnholsteredEvent (Variables|Default|GetWeaponOwnerReference) (EventDispatchers|CreateEvent :!event Custom|UnholsteredEvent))
  (EventDispatchers|BindEventtoOnPawnStartedEmote (Variables|Default|GetWeaponOwnerReference) (delegate Custom|EmoteStarted))
  (EventDispatchers|BindEventtoOnPawnStoppedEmote (Variables|Default|GetWeaponOwnerReference) (delegate Custom|EmoteStopped))
  (Default|BindEventtoOnAimDownSightsChanged (delegate Custom|AimDownSightsChangedEvent)))

(event Custom|InitializeProjectileTrajectory
  ;[0.04769 0.06954 0.4219 0.391 42] CHARGE UP AND TRAJECTORY
  ;;[0.2958 0.447 0.6042] InitializeProjectileTrajectory
  ;; Cache Player Pawn, setup and cache projectile variables
  (bind _asfort_player_pawn (Utilities|Casting|CastToFortPlayerPawn (Game|GetInstigator)))
  (bind _output_get (Variables|Cached|SetPlayerPawn _asfort_player_pawn))
  (if (Pawn|IsLocallyControlled _output_get)
    (Trajectory|SetupProjectileTrajectory)
    (Trajectory|CacheProjectileVariables)))

(event Custom|EmoteStopped
  (params (MontageItemDef :type Fort%20Item%20Definition%20Object%20Reference) (PawnEmoting :type Fort%20Pawn%20Object%20Reference))
  ;[0.04769 0.06954 0.4219 0.391 42] CHARGE UP AND TRAJECTORY
  ;;[0.4132 0.02567 1 1 32] VFX/SFX
  (CallFunction|UnholsteredEvent))

(event Custom|EmoteStarted
  (params (MontageItemDef :type Fort%20Item%20Definition%20Object%20Reference) (PawnEmoting :type Fort%20Pawn%20Object%20Reference))
  ;[0.04769 0.06954 0.4219 0.391 42] CHARGE UP AND TRAJECTORY
  ;;[0.4132 0.02567 1 1 32] VFX/SFX
  (CallFunction|HolsteredEvent))

(event Custom|AimDownSightsChangedEvent
  (params (AimDownSights :type Boolean))
  ;[0.04769 0.06954 0.4219 0.391 42] CHARGE UP AND TRAJECTORY
  ;;[0.4132 0.02567 1 1 32] VFX/SFX
  (if AimDownSights
    (CallFunction|StartIdleLoop)))

(event EventOnLocalClientReachedMaxCharge
  ;[0.04769 0.06954 0.4219 0.391 42] CHARGE UP AND TRAJECTORY
  ;;[0.4132 0.02567 1 1 32] VFX/SFX
  (Utilities|FlowControl|Sequence
    (:then_0
      ;;;[0 0.1784 1] Swap Looping Force Feedback at Max Charge
      (ForceFeedback|StopLocalForceFeedback (Variables|ForceFeedback|GetStartChargeForceFeedback) (Variables|ForceFeedback|GetStartChargeForceFeedbackTag))
      (ForceFeedback|PlayLocalForceFeedback (Variables|ForceFeedback|GetMaxChargeForceFeedback) (Variables|ForceFeedback|GetMaxChargeForceFeedbackTag) true))
    (:then_1
      (CallFunction|FeedbackFullCharge)
      (Components|Activation|Deactivate (Variables|Default|GetSFXc_ChargeUp))
      (Audio|PlaySoundatLocation (Variables|ChargeUp|GetSFX_FullChage) (Transformation|GetActorLocation self))
      (Niagara|SpawnSystemAttached (Variables|ChargeUp|GetVFX_FullyCharged) (Variables|Visual|GetWeaponMesh) "Muzzle")
      (Components|Activation|Activate (Variables|Default|GetSFXc_FullCharge) false))))

(event EventOnChargeUp
  ;[0.04769 0.06954 0.4219 0.391 42] CHARGE UP AND TRAJECTORY
  ;;[0.2958 0.447 0.6042] EventOnChargeUp (& Reset Charge)
  ;; Set bools to track charge up states for functionality and FX
  (|CallParentFunction)
  (Utilities|FlowControl|DoOnce false
    (:Completed
      (Variables|ChargeUp|SetbCharging true)
      (Variables|ChargeUp|SetbReachedMinCharge false) ; Can play charged weapon fire sound
      (Variables|ChargeUp|SetbReachedMaxCharge false))))

(event EventOnWeaponVisibilityChanged
  (params (bVisible :type Boolean) (bSetForLocalControllerOnly :type Boolean))
  (bind _returnvalue (not bVisible))
  ;[0.04769 0.06954 0.4219 0.391 42] CHARGE UP AND TRAJECTORY
  ;;[0.4132 0.02567 1 1 32] VFX/SFX
  ;;;[1 1 1] Idle audio
  (|CallParentFunction bVisible bSetForLocalControllerOnly)
  (if (and _returnvalue (HelperFunctions|IsLocalPawn (Variables|Actor|GetInstigator)))
    (CallFunction|ServerHolsterEvent _returnvalue)))

(event Custom|ResetCharge
  ;[0.04769 0.06954 0.4219 0.391 42] CHARGE UP AND TRAJECTORY
  ;;[0.2958 0.447 0.6042] EventOnChargeUp (& Reset Charge)
  ;; Set bools to track charge up states for functionality and FX
  (Variables|Firing|SetfProjectileVelocity 0.0)
  (Utilities|FlowControl|DoOnce false
    (:Completed
      (Variables|ChargeUp|SetbCharging true)
      (Variables|ChargeUp|SetbReachedMinCharge false) ; Can play charged weapon fire sound
      (Variables|ChargeUp|SetbReachedMaxCharge false))))

(event EventOnRemoteClientReachedMaxCharge
  ;[0.04769 0.06954 0.4219 0.391 42] CHARGE UP AND TRAJECTORY
  ;;[0.4132 0.02567 1 1 32] VFX/SFX
  (Components|Activation|Deactivate (Variables|Default|GetSFXc_ChargeUp))
  (Audio|PlaySoundatLocation (Variables|ChargeUp|GetSFX_FullChage) (Transformation|GetActorLocation self))
  (Niagara|SpawnSystemAttached (Variables|ChargeUp|GetVFX_FullyCharged) (Variables|Visual|GetWeaponMesh) "Muzzle")
  (Components|Activation|Activate (Variables|Default|GetSFXc_FullCharge) false))

(event Custom|ServerHolsterEvent
  (params (Holstered :type Boolean))
  ;[0.04769 0.06954 0.4219 0.391 42] CHARGE UP AND TRAJECTORY
  ;;[0.4132 0.02567 1 1 32] VFX/SFX
  ;;;[1 1 1] Idle audio
  (CallFunction|ClientHolsterEvent Holstered))

(event Custom|UnholsteredEvent
  ;[0.04769 0.06954 0.4219 0.391 42] CHARGE UP AND TRAJECTORY
  ;;[0.4132 0.02567 1 1 32] VFX/SFX
  ;;;[1 1 1] Idle audio
  (if (HelperFunctions|IsLocalPawn (Variables|Actor|GetInstigator))
    (CallFunction|ServerHolsterEvent false)))

(event Custom|ClientHolsterEvent
  (params (Holstered :type Boolean))
  ;[0.04769 0.06954 0.4219 0.391 42] CHARGE UP AND TRAJECTORY
  ;;[0.4132 0.02567 1 1 32] VFX/SFX
  ;;;[1 1 1] Idle audio
  (Variables|Default|SetbHolstered Holstered))

(event EventOnLocalClientReachedMinCharge
  ;[0.04769 0.06954 0.4219 0.391 42] CHARGE UP AND TRAJECTORY
  ;;[0.4132 0.02567 1 1 32] VFX/SFX
  (Variables|ChargeUp|SetbReachedMinCharge true) ; Can play charged weapon fire sound
  ;;;[0.4132 0.02567 1] FX: Charge up Sounds and visuals
  (Components|Activation|Activate (Variables|Default|GetSFXc_ChargeUp) false)
  (Niagara|SetNiagaraVariable(Float) (Variables|Default|GetNS_SawBlade_Chargeup) "User.ChargeUp" 1.0))

(event Custom|IncreaseProjectileCharge
  ;[0.04769 0.06954 0.4219 0.391 42] CHARGE UP AND TRAJECTORY
  ;;[0 0.07377 1] IncreaseProjectileCharge
  ;; Calculate velocity and update the trajectory visuals
  (Variables|Firing|SetfProjectileVelocity (Math|Float|SelectFloat (Trajectory|CalculateProjectileVelocity) (ScalableFloat|GetValueatLevel (Variables|Firing|GetfMaxProjectileSpeed)) (<= (Variables|Firing|GetfProjectileVelocity) (ScalableFloat|GetValueatLevel (Variables|Firing|GetfMaxProjectileSpeed)))))
  (Trajectory|CalculateTrajectorySpline))

(event Custom|StartIdleLoop
  (bind _sfxc_passive (Variables|Default|GetSFXc_Passive))
  ;[0.04769 0.06954 0.4219 0.391 42] CHARGE UP AND TRAJECTORY
  ;;[0.4132 0.02567 1 1 32] VFX/SFX
  ;;;[1 1 1] Idle audio
  (Networking|CanExecuteCosmeticEvents
    (:True
      (if (not (Components|Activation|IsActive _sfxc_passive))
        (Audio|Components|Audio|Play _sfxc_passive 0.0)))))

(event Custom|HolsteredEvent
  ;[0.04769 0.06954 0.4219 0.391 42] CHARGE UP AND TRAJECTORY
  ;;[0.4132 0.02567 1 1 32] VFX/SFX
  ;;;[1 1 1] Idle audio
  (if (HelperFunctions|IsLocalPawn (Variables|Actor|GetInstigator))
    (CallFunction|ServerHolsterEvent true)))

(event EventOnRemoteClientReachedMinCharge
  ;[0.04769 0.06954 0.4219 0.391 42] CHARGE UP AND TRAJECTORY
  ;;[0.4132 0.02567 1 1 32] VFX/SFX
  (CallFunction|StartMeleeDamageAudioVisuals)
  (bind _returnvalue (Utilities|Time|SetTimerbyEvent (EventDispatchers|CreateEvent :!event Custom|StartMeleeDamageAudioVisuals) (ScalableFloat|GetValueatLevel (Variables|Melee|GetHF_MeleeDamageRate)) true))
  (Variables|Melee|SetTH_MeleeAVIndicator _returnvalue)
  (Variables|ChargeUp|SetbReachedMinCharge true) ; Can play charged weapon fire sound
  ;;;[0.4132 0.02567 1] FX: Charge up Sounds and visuals
  (Components|Activation|Activate (Variables|Default|GetSFXc_ChargeUp) false)
  (Niagara|SetNiagaraVariable(Float) (Variables|Default|GetNS_SawBlade_Chargeup) "User.ChargeUp" 1.0))

(event Custom|StopSawBladeGunAudio
  ;[0.04769 0.06954 0.4219 0.391 42] CHARGE UP AND TRAJECTORY
  ;;[0.4132 0.02567 1 1 32] VFX/SFX
  ;;;[1 1 1] Idle audio
  (Networking|CanExecuteCosmeticEvents
    (:True
      (if (Components|Activation|IsActive (Variables|Default|GetSFXc_Passive))
        (Audio|Components|Audio|FadeOut (Variables|Default|GetDamageAudioComp) 0.1 0.0 "Linear" :!class /Script/Engine.AudioComponent)
        (for _array_element (Variables|Default|GetLatestDamageOneShotAudioComps)
          (Audio|Components|Audio|FadeOut _array_element 0.1 0.0 "Linear" :!class /Script/Engine.AudioComponent)
          (bind _weaponownerreference (Variables|Default|GetWeaponOwnerReference)
            (:then
              (Audio|SpawnSoundAttached "/SawbladeGun/Items/SawbladeGun/Sounds/SawBladeGun_Unequip_Cue.SawBladeGun_Unequip_Cue" (Variables|/Script/Engine.Actor|GetRootComponent _weaponownerreference)))
            (:else)))))))

(event EventTick
  (params (DeltaSeconds :type Float%20%28single-precision%29))
  (bind _returnvalue (Game|GetInstigatorController))
  ;[0.04769 0.06954 0.4219 0.391 42] CHARGE UP AND TRAJECTORY
  ;;[0 0.8169 1] Event Tick
  ;; When there is a trajectory indicator, update based on the current charge level
  (|CallParentFunction DeltaSeconds)
  (Variables|Trajectory|GetSawBladeCustomTrajectory
    (:then
      (Utilities|IsValid _returnvalue
        (:Is%20Valid
          (if (Pawn|IsLocalPlayerController _returnvalue)
            (CallFunction|IncreaseProjectileCharge)))))
    (:else)))

(event Custom|FeedbackFullCharge
  ;[0.04769 0.06954 0.4219 0.391 42] CHARGE UP AND TRAJECTORY
  ;;[0.4132 0.02567 1 1 32] VFX/SFX
  ;;;[0.4132 0.02567 1] Swap GCNs at Full Charge
  (bind _playerpawn (Variables|Cached|GetPlayerPawn)
    (:then
      (GameplayCue|AddGameplayCueOnActor(Looping) _playerpawn (Variables|ChargeUp|GetGCTag_ChargeUpFull) (Ability|GameplayCue|MakeGameplayCueParameters 0.0))
      (GameplayCue|RemoveGameplayCueOnActor(Looping) _playerpawn (Variables|ChargeUp|GetGCTag_ChargeUp) (Ability|GameplayCue|MakeGameplayCueParameters 0.0)))
    (:else)))

(event FortWeapon|EventOnLocalUnEquip
  ;[0.04769 0.06954 0.4219 0.391 42] CHARGE UP AND TRAJECTORY
  ;;[1 0.5272 0.04774] EventOnLocalUnEquip
  ;; End the charge which cleans up variables and visuals when the weapon is no longer equpped
  (|CallParentFunction)
  (CallFunction|StopSawBladeGunAudio)
  (CallFunction|EndCharge))

(event EventOnEndCharge
  (bind _returnvalue (Ability|GameplayCue|MakeGameplayCueParameters 0.0))
  ;[0.04769 0.06954 0.4219 0.391 42] CHARGE UP AND TRAJECTORY
  ;;[1 0.5272 0.04774] EventOnEndCharge
  ;; Reset variables and start timer to initiate a new chargeup
  (|CallParentFunction)
  (Utilities|FlowControl|Sequence
    (:then_0
      ;;;[1 0.5272 0.04774] Sequence 0: Reset variables and destroy the trajectory object
      (Variables|Firing|SetfProjectileVelocity 0.0)
      (Variables|ChargeUp|SetbCharging false)
      (bind _sawbladecustomtrajectory (Variables|Trajectory|GetSawBladeCustomTrajectory)
        (:then
          (Actor|DestroyActor _sawbladecustomtrajectory)
          (CallFunction|ResetCharge))
        (:else)))
    (:then_1
      ;;;[1 0.5272 0.04774] Sequence 1: Disable melee damage
      (Utilities|Time|ClearandInvalidateTimerbyHandle (Variables|Melee|GetMeleeDamageTimer))
      (Utilities|Time|ClearandInvalidateTimerbyHandle (Variables|Melee|GetTH_MeleeAVIndicator))
      (Variables|Melee|SetRepSetMeleeVFX false))
    (:then_2
      ;;;[0.4132 0.02567 1] Sequence 2: Disable SFX and VFX
      (Components|Activation|Deactivate (Variables|Default|GetSFXc_ChargeUp))
      (Niagara|SetNiagaraVariable(Float) (Variables|Default|GetNS_SawBlade_Chargeup) "User.ChargeUp" 0.0)
      ;;;[0.4132 0.02567 1] Stop Force Feedback
      (ForceFeedback|StopLocalForceFeedback (Variables|ForceFeedback|GetStartChargeForceFeedback) (Variables|ForceFeedback|GetStartChargeForceFeedbackTag))
      (ForceFeedback|StopLocalForceFeedback (Variables|ForceFeedback|GetMaxChargeForceFeedback) (Variables|ForceFeedback|GetMaxChargeForceFeedbackTag))
      ;;;[0.4132 0.02567 1] Remove Gameplay Cues for charging
      (bind _playerpawn (Variables|Cached|GetPlayerPawn)
        (:then
          (GameplayCue|RemoveGameplayCueOnActor(Looping) _playerpawn (Variables|ChargeUp|GetGCTag_ChargeUp) _returnvalue)
          (GameplayCue|RemoveGameplayCueOnActor(Looping) _playerpawn (Variables|ChargeUp|GetGCTag_ChargeUpFull) _returnvalue))
        (:else)))))

(event Custom|StartMeleeDamageAudioVisuals
  ;[0.04769 0.06954 0.4219 0.391 42] Melee Damage
  ;;[0.4132 0.02567 1] StartMeleeDamageAudioVisuals
  (Fortnite|ShowSoundIndicatorLocalforTeam :TrackedActor self :Location (Transformation|GetSocketLocation (Variables|Visual|GetWeaponMesh) "Muzzle") :MaxAudibleDistance (ScalableFloat|GetValueatLevel (Variables|Melee|GetHF_AVIndicator_VisibleDistance)) :IndicatorType "Gunshot" :Team (Team|GetActorTeam self) :Affiliations (Utilities|Array|MakeArray "Hostile" "Neutral") :Instigator self :Tint "(R=0.000000,G=0.000000,B=0.000000,A=0.000000)" :AttenuationMaxRangeMultiplier 1.0 :AttenuationIntensityMultiplier 1.0))

(event Custom|StartMelee
  ;[0.04769 0.06954 0.4219 0.391 42] Melee Damage
  ;;[0.1083 1 0.0636] StartMelee
  ;; Check if melee Hotfix is enabled and start damage check timer
  (if (ScalableFloat|GetValueasBool (Variables|Melee|GetHF_MeleeEnabled))
    ;;;[0.5 0.4911 0] The weapon's owning client handles updating the Melee function. This is necessary because the server does NOT have an accurate muzzle location.
    ;;; Weapons fire projectiles the same way (with the owning client telling the server the location and direction of the shot.)
    (if (HelperFunctions|IsLocalPawn (Variables|Actor|GetInstigator))
      (CallFunction|OnTimerMeleeDamageRate) ; call initially for immediate responsiveness
      (bind _returnvalue (Utilities|Time|SetTimerbyEvent (EventDispatchers|CreateEvent :!event Custom|OnTimerMeleeDamageRate) (ScalableFloat|GetValueatLevel (Variables|Melee|GetHF_MeleeDamageRate)) true))
      (Variables|Melee|SetMeleeDamageTimer _returnvalue))))

(event Custom|OnTimerMeleeDamageRate
  ;[0.04769 0.06954 0.4219 0.391 42] Melee Damage
  ;;[0 0.8169 1] OnTimerMeleeDamageRate
  ;; Get Client muzzle position and send to server for damage
  (if (HelperFunctions|IsLocalPawn (Variables|Actor|GetInstigator))
    (CallFunction|ServerMeleeDamage (Transformation|GetSocketLocation (Variables|Visual|GetWeaponMesh) "Muzzle"))))

(event Custom|ServerMeleeDamage
  (params (MuzzleLocation :type Vector))
  ;[0.04769 0.06954 0.4219 0.391 42] Melee Damage
  ;;[0.4062 0.206 0.07282] ServerMeleeDamage
  ;; Start an async overlap to detect melee victims
  (Variables|Melee|SetRepSetMeleeVFX false)
  (bind _instigator (Variables|Actor|GetInstigator)
    (:then
      (Utilities|IsValid (Ability|GetFortAbilitySystemComponent _instigator)
        (:Is%20Valid
          (bind _outhits (Collision|MultiSphereTraceByChannel MuzzleLocation MuzzleLocation (ScalableFloat|GetValueatLevel (Variables|Melee|GetHF_MeleeDamageRadius)) "TraceTypeQuery5" false (Utilities|Array|MakeArray self (Actor|GetOwner))))
          (for _array_element _outhits
            (if (Collision|BreakHitResult _array_element)
              (bind _hitactorfortabilitysystem (Utilities|CheckforValidMeleeHit _instigator (.HitActor (Collision|BreakHitResult _array_element)))
                (:ValidMeleeHit
                  (Variables|Melee|SetRepSetMeleeVFX true)
                  (bind _output_get (Variables|Melee|SetHarvestDamageContext (Class|AthenaFunctionLibrary|MakeSimpleFortEffectContext _instigator self))) ; We need to store this value temporarily so we can add the hit result to it\nand then pass it into ApplyGameplayEffectToTarget().
                  (Ability|EffectContext|AddHitResult _output_get _array_element)
                  (GameplayEffects|ApplyGameplayEffectToTarget (Ability|GetFortAbilitySystemComponent _instigator) "/SawbladeGun/Items/SawbladeGun/GE_SawbladeGun_ApplyMeleeHarvesting.GE_SawbladeGun_ApplyMeleeHarvesting_C" _hitactorfortabilitysystem 0.0 _output_get) ; Deal Damage
                  (CallFunction|MultiCastDamageAudio _array_element)
                  (bind _asfort_pawn (Utilities|Casting|CastToFortPawn (.HitActor (Collision|BreakHitResult _array_element))))
                  (Character|LaunchCharacter _asfort_pawn (Melee|CalculateMeleeKnockback) false true))))))))
    (:else)))

(event FortWeapon|EventOnSpawnProjectile
  (params (SpawnProjectile :type Fort%20Projectile%20Base%20Object%20Reference))
  ;[1 1 1] Initialize projectile with speed based on how charged up the weapon was
  (bind _asprj_athena_sawblade (Utilities|Casting|CastToPrj_Athena_Sawblade SpawnProjectile))
  (Class|PrjAthenaSawblade|Init _asprj_athena_sawblade (Variables|ChargeUp|GettempCharge) (Utilities|RemapFloat (Variables|ChargeUp|GettempCharge) (Math|Vector2D|MakeVector2D 0.0 (Variables|ChargeUp|GettempMaxCharge)) "(X=-1000.000000,Y=4500.000000)" true) (Math|Vector|Normalize2D(Vector) (Transformation|GetVelocity _asprj_athena_sawblade))))

(event Custom|MultiCastDamageAudio
  (params (HitResult :type Hit%20Result%20Structure))
  (bind (_bblockinghit _binitialoverlap _time _distance _location _impactpoint _normal _impactnormal _physmat _hitactor) (Collision|BreakHitResult HitResult))
  (bind (_value _returnvalue) (Utilities|Map|Find (Variables|Default|GetDamageLoopAudioMap) (Variables|Default|GetMeleePhysicalSurfaceType)))
  (bind _latestdamageoneshotaudiocomps (Variables|Default|GetLatestDamageOneShotAudioComps))
  (bind _audiodamageoneshotindex (Variables|Default|GetAudioDamageOneShotIndex))
  (bind _audiodamageoneshotindex_2 (Variables|Default|GetAudioDamageOneShotIndex))
  ;[0.9333 1 0.3529] Audio - melee damage
  (Networking|CanExecuteCosmeticEvents
    (:True
      (Utilities|FlowControl|Sequence
        (:then_0
          ;;[0.9333 1 0.3529] Check for physical surface on hit. Set customized surface type based on actor (ie Vehicles, Pawns, etc.), everything else should have fort physical surface.
          (Utilities|IsValid _hitactor
            (:Is%20Valid
              (Utilities|Casting|CastToFortAthenaVehicle _hitactor
                (:then
                  (Variables|Default|SetMeleePhysicalSurfaceType "SurfaceType3"))
                (:CastFailed
                  (Utilities|Casting|CastToFortPlayerPawn _hitactor
                    (:then
                      (Variables|Default|SetMeleePhysicalSurfaceType "SurfaceType5"))
                    (:CastFailed
                      (Variables|Default|SetMeleePhysicalSurfaceType (Collision|GetFortPhysicalSurface HitResult)))))))
            (:Is%20Not%20Valid
              (Variables|Default|SetMeleePhysicalSurfaceType "SurfaceType_Default")))
          ;;[0.9333 1 0.3529] Play sound after setting physical surface. Due to race condition.
          (if (FortWeapon|IsEquipped)
            (bind _returnvalue_2 (Audio|SpawnSoundAttached (Variables|Default|GetDamageAudioHit) (Variables|Visual|GetWeaponMesh)))
            (if (< _audiodamageoneshotindex_2 2)
              (Math|Integer|IncrementInt _audiodamageoneshotindex_2
                (:%20%20
                  _))
              (else
                (Variables|Default|SetAudioDamageOneShotIndex 0)))
            (bind _result (ArrayMacros|SafeArrayGet _latestdamageoneshotaudiocomps _audiodamageoneshotindex)
              (:%20%20
                (Audio|Components|Audio|FadeOut _result 0.3 0.0 "Linear" :!class /Script/Engine.AudioComponent)
                (Utilities|Array|SetArrayElem _latestdamageoneshotaudiocomps _audiodamageoneshotindex _returnvalue_2 true) ; Object pooling for damage sounds. \nOne-shots are longer, and we want to \noverlap them but also fade them all out \nquickly if switching weapons mid-damage.
                (Audio|Parameter|SetIntegerParameter _returnvalue_2 "PhysicalSurface" (Math|Integer|SelectInt _value 0 _returnvalue) :!class /Script/AudioExtensions.AudioParameterControllerInterface)
                (Variables|Default|GetDamageAudioComp
                  (:then
                    _)
                  (:else
                    (bind _returnvalue_3 (Audio|SpawnSoundAttached :Sound (Variables|Default|GetDamageAudioLoop) :AttachToComponent (Variables|Default|GetSFXc_ChargeUp) :AttachPointName "None" :LocationType "KeepRelativeOffset" :bStopWhenAttachedToDestroyed true :VolumeMultiplier 1.0 :PitchMultiplier 1.0 :StartTime 0.0 :bAutoDestroy true))
                    (Audio|Components|Audio|FadeIn _returnvalue_3 0.1 1.0 0.0 "Linear" :!class /Script/Engine.AudioComponent)
                    (Variables|Default|SetDamageAudioComp _returnvalue_3)))
                (Melee|SetDamageAudioSurfaceType (Variables|Default|GetDamageAudioComp))))))
        (:then_1
          (Utilities|FlowControl|RetriggerableDelay (+ (ScalableFloat|GetValueatLevel (Variables|Melee|GetHF_MeleeDamageRate)) 0.1))
          (Audio|Components|Audio|FadeOut (Variables|Default|GetDamageAudioComp) 0.1 0.0 "Linear" :!class /Script/Engine.AudioComponent))))))

(event Custom|OnLocalAmmoChange
  (params (LocalCount :type Integer) (LocalRemaining :type Integer))
  ;[0.4062 0.206 0.07282] OnLocalAmmoChanged
  ; Check the remaining ammo to see if it is expended then Send the Gameplay Event for the GA which removes the item when expended
  (if (<= LocalRemaining 0)
    (bind _asfort_player_pawn (Utilities|Casting|CastToFortPlayerPawn (Game|GetInstigator)))
    (Ability|SendGameplayEventtoActor _asfort_player_pawn (Variables|Reloading|GetEvent_AmmoDepletedRemoveWeapon) (Utilities|Struct|MakeGameplayEventData "(TagName=\"\")" self (Variables|Cached|GetPlayerPawn))) ; will remove the weapon from inventory
    (EventDispatchers|UnbindEventfromOnLocalAmmoChanged (EventDispatchers|CreateEvent :!event Custom|OnLocalAmmoChange))))

(event EventEndPlay
  (params (EndPlayReason :type EEndPlayReason%20Enum :default "Destroyed"))
  ;[1 0.5272 0.04774] EventEndPlay
  ; Final cleansup of BP: Unbind, Clear timers, etc.
  (|CallParentFunction EndPlayReason)
  (Utilities|FlowControl|Sequence
    (:then_0
      (EventDispatchers|UnbindEventfromOnLocalAmmoChanged (EventDispatchers|CreateEvent :!event Custom|OnLocalAmmoChange)))
    (:then_1
      (Utilities|Time|ClearandInvalidateTimerbyHandle (Variables|Melee|GetMeleeDamageTimer))
      (Utilities|Time|ClearandInvalidateTimerbyHandle (Variables|Melee|GetTH_MeleeAVIndicator)))))

(event Custom|Play2DHitNotification
  (Networking|CanExecuteCosmeticEvents
    (:True
      (if (Pawn|IsLocallyControlled (Game|GetInstigator))
        (Audio|PlaySound2D "/SawbladeGun/Items/SawbladeGun/Sounds/SawbladGun_HitMarker_Death_Cue.SawbladGun_HitMarker_Death_Cue")))))

(fn CacheProjectileVariables
  (bind _playerpawn (Variables|Cached|GetPlayerPawn)
    (:then
      (bind _trajectorydata (Class|FortKismetLibrary|GetProjectileTrajectoryData _playerpawn (Variables|Default|GetProjectileClass) 0.0 1.0))
      (Variables|Trajectory|SetfProjectile_MaxSpeed (Utilities|Struct|BreakFortProjectileTrajectoryData _trajectorydata))
      (Variables|Trajectory|SetfProjectile_GravityZ (.GravityZ (Utilities|Struct|BreakFortProjectileTrajectoryData _trajectorydata)))
      (Variables|Trajectory|SetfProjectile_Extent (Math|Float|Max(Float) (.CapsuleRadius (Utilities|Struct|BreakFortProjectileTrajectoryData _trajectorydata)) (.CapsuleHalfHeight (Utilities|Struct|BreakFortProjectileTrajectoryData _trajectorydata))))
      (Variables|Trajectory|SetfProjectile_Bounciness (.Bounciness (Utilities|Struct|BreakFortProjectileTrajectoryData _trajectorydata)))
      (Variables|Trajectory|SetfProjectile_Friction (.Friction (Utilities|Struct|BreakFortProjectileTrajectoryData _trajectorydata)))
      (Variables|Trajectory|SetProjectile_CollisionChannel (.CollisionObjectType (Utilities|Struct|BreakFortProjectileTrajectoryData _trajectorydata))))
    (:else)))

(fn CalculateMeleeKnockback
  ;[0.4729 0.7399 1] CalculateMeleeKnockback
  ; Calculate the knockback bounce from variables and weapon look vector
  ;;[1 0.3702 0.9557] Return bounce vector
  (return (* (+ (* (Transformation|GetRightVector (Variables|Visual|GetWeaponMesh)) (ScalableFloat|GetValueatLevel (Variables|Melee|GetHF_MeleeBounceMagnitude))) (Math|Vector|MakeVector 0.0 0.0 (Variables|Melee|GetKnockbackImpulseZOffset%20Const))) (Variables|Melee|GetKnockbackPawnImpulseMultiplier%20Const))))

(fn CalculateProjectileVelocity
  (return (FortWeapon|GetProjectileSpeed (FortWeapon|GetChargePercent))))

(fn CalculateTrajectorySpline
  (bind (_startlocation _initialvelocity _projectilevelocity _outrotation) (Trajectory|SetupProjectileVariables))
  (FortWeapon|UpdateDamageStartLocation)
  (bind _playerpawn (Variables|Cached|GetPlayerPawn)
    (:then
      (bind (_outhitresult _outsplinepoints _outsplinetangents) (Projectile|CalculateProjectileTrajectorySplinewithHitProfile _playerpawn _startlocation _initialvelocity _projectilevelocity (Variables|Trajectory|GetfProjectile_GravityZ) (Variables|Trajectory|GetfProjectile_Friction) (Variables|Trajectory|GetfProjectile_Bounciness) (Variables|Trajectory|GetfProjectile_Timestep) (Variables|Trajectory|GetfProjectile_Extent) (Variables|Trajectory|GetTrajectoryCollisionProfile) (Class|FortPawn|GetProjectileIgnoredActors _playerpawn) 1 30 (Math|Float|Lerp 3.5e+02 1e+03 (FortWeapon|GetChargePercent)) 3e+01))
      (bind _sawbladecustomtrajectory (Variables|Trajectory|GetSawBladeCustomTrajectory)
        (:then
          (Trajectory|SetTrajectorySpline _sawbladecustomtrajectory _outsplinepoints _outsplinetangents)
          (Variables|/SawbladeGun/Items/SawbladeGun/BP_SawbladeGun_ProjectileTrajectory.BP_SawbladeGun_ProjectileTrajectory_C|SetAimRotation _outrotation _sawbladecustomtrajectory)
          (return _outhitresult))
        (:else)))
    (:else)))

(fn MeleeOverlapShouldCancel
  (return false))

(fn OnRep_ArrayElement)

(fn OnRep_bDamageAudioFaded)

(fn OnRep_bHolstered
  (if (Variables|Default|GetbHolstered)
    (CallFunction|StopSawBladeGunAudio)
    (elif (GameplayTags|HasAnyMatchingGameplayTags (Utilities|Casting|CastToFortPlayerPawn (Variables|Actor|GetInstigator)) (Variables|Default|GetTC_ActionsThatPreventIdleAudio))
      _
      (else
        (CallFunction|StartIdleLoop)))))

(fn OnRep_bPlayDamageAudio
  (Networking|CanExecuteCosmeticEvents
    (:True
      (if (Variables|Default|GetbPlayingDamageAudio)
        (Variables|Default|GetDamageAudioComp
          (:then)
          (:else
            (bind _returnvalue (Audio|SpawnSoundAttached :Sound (Variables|Default|GetDamageAudioLoop) :AttachToComponent (Variables|Default|GetSFXc_ChargeUp) :AttachPointName "None" :LocationType "KeepRelativeOffset" :bStopWhenAttachedToDestroyed true :VolumeMultiplier 1.0 :PitchMultiplier 1.0 :StartTime 0.0 :bAutoDestroy true))
            (Variables|Default|SetDamageAudioComp _returnvalue)))
        (elif (Audio|Components|Audio|IsPlaying (Variables|Default|GetDamageAudioComp))
          (Audio|Components|Audio|FadeOut (Variables|Default|GetDamageAudioComp) 0.1 0.0 "Linear" :!class /Script/Engine.AudioComponent))))))

(fn OnRep_DamageCounter
  (if (== (Variables|Default|GetDamageCounter) 0)
    (Utilities|FlowControl|Sequence
      (:then_0
        _)
      (:then_1
        _))
    (Utilities|FlowControl|Gate true
      (:Exit
        (if (>= (Variables|Default|GetDamageCounter) 2)
          (Audio|Components|Audio|FadeOut (Variables|Default|GetDamageAudioComp) 0.1 0.0 "Linear" :!class /Script/Engine.AudioComponent))))))

(fn OnRep_RepSetMeleeVFX
  (bind _vfx_niagra_damage (Variables|Default|GetVFX_Niagra_Damage))
  ;[0.4132 0.02567 1] OnRep SetMeleeVFX (CLIENT)
  ; Activates or deactivates the melee VFX
  (Networking|CanExecuteCosmeticEvents
    (:True
      (if (Variables|Melee|GetRepSetMeleeVFX)
        (Components|Activation|Activate _vfx_niagra_damage false)
        (Effects|Components|ParticleSystem|SetIntParameter _vfx_niagra_damage "PhysicalSurface" 0)
        (else
          (Components|Activation|Deactivate _vfx_niagra_damage))))))

(fn SetDamageAudioSurfaceType
  (params (AudioComponent :type Audio%20Component%20Object%20Reference))
  (bind _meleephysicalsurfacetype (Variables|Default|GetMeleePhysicalSurfaceType))
  (bind _damageloopaudiomap (Variables|Default|GetDamageLoopAudioMap))
  (bind (_value _returnvalue) (Utilities|Map|Find _damageloopaudiomap _meleephysicalsurfacetype))
  ;[1 1 1] Find Fort Physical Melee Surface type that match audio map
  (if _returnvalue
    (if (not (== _value (Variables|Default|GetAudioSurfaceIndex)))
      (Variables|Default|SetAudioSurfaceIndex _value)
      (Audio|Parameter|SetIntegerParameter AudioComponent "PhysicalSurface" _value :!class /Script/AudioExtensions.AudioParameterControllerInterface))
    (elif (== (Variables|Default|GetAudioSurfaceIndex) 0)
      _
      (else
        (Variables|Default|SetAudioSurfaceIndex 0)
        (Audio|Parameter|SetIntegerParameter AudioComponent "PhysicalSurface" 0 :!class /Script/AudioExtensions.AudioParameterControllerInterface)))))

(fn SetupProjectileTrajectory
  ;[1 1 1] SetupProjectileTrajectory
  ; Spawn a trajectory actor, store a local ref, and set tick requirment
  (bind _playerpawn (Variables|Cached|GetPlayerPawn)
    (:then
      (bind _returnvalue (Game|SpawnActorfromClass (Variables|Trajectory|GetTrajectoryClass) (Transformation|GetActorTransform _playerpawn) "AlwaysSpawn" "SelectDefaultAtRuntime" _playerpawn _playerpawn))
      (bind _asbp_sawblade_gun_projectile_trajectory (Utilities|Casting|CastToBP_SawbladeGun_ProjectileTrajectory _returnvalue))
      (bind _output_get (Variables|Trajectory|SetSawBladeCustomTrajectory _asbp_sawblade_gun_projectile_trajectory))
      (Actor|Tick|AddTickPrerequisiteActor _output_get self))
    (:else)))

(fn SetupProjectileVariables
  (return (Transformation|GetSocketLocation (Variables|Visual|GetWeaponMesh) "Muzzle") (* (Math|Rotator|GetRotationXVector (Math|Rotator|MakeRotator (.roll (Math|Conversions|RotationFromXVector (FortWeapon|GetDamageStartLocation -1))) (+ (.pitch (Math|Conversions|RotationFromXVector (FortWeapon|GetDamageStartLocation -1))) (FortWeapon|GetProjectilePitchOffset)) (.yaw (Math|Conversions|RotationFromXVector (FortWeapon|GetDamageStartLocation -1))))) (Variables|Firing|GetfProjectileVelocity)) (Variables|Firing|GetfProjectileVelocity) (Math|Rotator|MakeRotator (.roll (Math|Conversions|RotationFromXVector (FortWeapon|GetDamageStartLocation -1))) (+ (.pitch (Math|Conversions|RotationFromXVector (FortWeapon|GetDamageStartLocation -1))) (FortWeapon|GetProjectilePitchOffset)) (.yaw (Math|Conversions|RotationFromXVector (FortWeapon|GetDamageStartLocation -1))))))

(fn ConstructionScript
  (|CallParentFunction))
