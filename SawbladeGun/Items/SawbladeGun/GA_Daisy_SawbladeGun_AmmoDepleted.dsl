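; Event-triggered gameplay ability that discards the Daisy Sawblade Gun when it runs out of ammo.
; Activation casts the event's instigator and target to the weapon and pawn, then plays the out-of-ammo sound and toss montage.
; The montage's Toss notify tag ends the ability.
; OnEndAbility removes the weapon from the player's inventory by GUID on authority only, without checking whether the ability was cancelled.
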(event Ability|EventActivateAbilityFromEvent
  (params (EventData :type Gameplay%20Event%20Data%20Structure%20%28by%20ref%29))
  (bind (_eventtag _instigator _target) (Utilities|Struct|BreakGameplayEventData EventData))
  ;[0.4132 0.02567 1] EventActivateAbilityFromEvent
  ; Check the ammo and determine whether to remove the weapon with cooresponding FX
  (|CallParentFunction EventData) ; call parent
  (Utilities|FlowControl|Sequence
    (:then_0
      ;;[0.208 0.462 0.7396] Sequence (0) Store the pawn and weapon which activated the ability to check ammo
      (bind _asb_sawblade_gun_athena (Utilities|Casting|CastToB_SawbladeGun_Athena _instigator))
      (Variables|Default|SetThisWeapon _asb_sawblade_gun_athena)
      (bind _asfort_player_pawn (Utilities|Casting|CastToFortPlayerPawn _target))
      (Variables|Default|SetFortPawn _asfort_player_pawn))
    (:then_1
      ;;[0.4132 0.02567 1] Sequence (1) VFX and SFX of removing the weapon
      (Variables|Default|GetThisWeapon
        (:then
          ;;;[0.4132 0.02567 1] Play SFX and Play montage of throwing away the weapon
          (bind _fortpawn (Variables|Default|GetFortPawn)
            (:then
              (Audio|PlaySoundatLocation (Variables|Default|GetOutOfAmmoSound) (Transformation|GetActorLocation _fortpawn))
              (bind (_asynctaskproxy _notifytag) (Ability|Tasks|PlayMontageAndWaitNotify "None" (Variables|Default|GetOutOfAmmoMontage))
                (:OnInterrupted
                  _)
                (:OnCancelled
                  _)
                (:OnNotifyBegin
                  (switch Utilities|FlowControl|Switch|SwitchonGameplayTag _notifytag
                    (:Event.Athena.Item.SawbladeGun.Toss
                      ;;;[1 0.3073 0.3805] EndAbility
                      (Ability|EndAbility))))))
            (:else
              _)))
        (:else
          _)))) ; (0) store references\n(1) check ammo and maybe remove weapon
)

(event Ability|EventOnEndAbility
  (params (bWasCancelled :type Boolean))
  ;[1 0.4428 0] Event OnEndAbility
  ; Remove the Item from the player's inventory if pawn and weapon is valid
  (|CallParentFunction bWasCancelled)
  ;;[0.3021 0.1061 0] Remove expended weapon from inventory
  (bind _thisweapon (Variables|Default|GetThisWeapon)
    (:then
      (bind _fortpawn (Variables|Default|GetFortPawn)
        (:then
          (Ability|HasAuthority
            (:True
              (Utilities|IsValid (Fortnite|GetFortPlayerControllerfromActor _fortpawn)
                (:Is%20Valid
                  (Inventory|RemoveItemFromPlayerByGuid (Fortnite|GetFortPlayerControllerfromActor _fortpawn) (FortWeapon|GetInventoryGUID _thisweapon) 1 true))))))
        (:else)))
    (:else)))
