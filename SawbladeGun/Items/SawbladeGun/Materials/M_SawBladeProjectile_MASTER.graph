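# Master material for the sawblade projectile that spins the blade in the vertex shader.
# The shared Fortnite gun master function supplies the base material attributes.
# Time and a speed parameter drive RotateAboutAxis around the object pivot on a local-space axis parameter, added to the base World Position Offset.
# FixRotateAboutAxisNormals uses the same axis and angle to replace the Normal attribute so shading follows the rotation.
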Material /SawbladeGun/Items/SawbladeGun/Materials/M_SawBladeProjectile_MASTER
graph {
 properties {
  bUseMaterialAttributes=true
  bUsedWithMeshDeformer=true
  bUsedWithSkeletalMesh=true
  pixelDepthOffsetMode=PDOM_Legacy
  usedShadingModels=Default Lit
 }
 nodes [
  n1 = MaterialFunctionCall!(UV 1 coordinates, MuzzleFalloff) {materialFunction=/Game/Packages/Fortnite_UniqueMaterials_1/Weapons/M_GunMasterShaderFunction}
  GetMaterialAttributes(None=@n1.Material Result) {attributeGetTypes [#1, #2]}
  n2 = Transform(None=VectorParameter() {parameterName=LocalRotationAxis, defaultValue=<rgba 1 0 0 1>}.RGB) {transformSourceType=TRANSFORMSOURCE_Local}
  n3 = Multiply(A=Time(), B=ScalarParameter() {defaultValue=-12, parameterName=BladeRotationSpeed})
  n4 = MaterialFunctionCall() {materialFunction=&ObjectPivotPoint}
  n5 = WorldPosition()
  SetMaterialAttributes!(
   MaterialAttributes=@n1.Material Result
   World Position Offset=Add(
    A=@n1.World Position Offset
    B=RotateAboutAxis(NormalizedRotationAxis=@n2, RotationAngle=@n3, PivotPoint=@n4.Object Pivot Location, Position=@n5.XYZ)
   )
   Normal=MaterialFunctionCall(Rotation Axis=@n2, Rotation Angle=@n3) {materialFunction=&FixRotateAboutAxisNormals}.New Tangent Space Vertex Normal
  ) {attributeSetTypes [#1, #2]}
  VertexNormalWS()
 ]
}
