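; Projectile Blueprint for the Sawblade Gun: a blade that travels along surfaces, damaging and knocking back what it hits.
; The server drives hit and overlap handling, stopping at buildings to damage them on a timer, and queues pawn launches for Tick.
; Pickaxe hits or blocks redirect the blade, changing instigator and restoring lifespan up to a reflection limit.
; Stuck checks shorten its life; surface-based FX, sound indicators and an end-of-life wobble provide feedback.
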(event EventBeginPlay
  ;[0.1083 1 0.0636] EventBeginPlay
  ; Bind events for destroy
  (Utilities|FlowControl|Sequence
    (:then_0
      ;;[0.1083 1 0.0636] Sequence [0]: Bind events and core behavior setup
      (Game|BindEventtoOnDestroyed (EventDispatchers|CreateEvent :!event Custom|EventOnDestroyed))
      (Networking|SwitchHasAuthority
        (:Authority
          (bind _returnvalue (Utilities|Time|SetTimerbyEvent (EventDispatchers|CreateEvent :!event Custom|LifeSpanNearEndWarning) (- (Variables|Actor|GetInitialLifeSpan) (Variables|VFX|GetLifeSpanEndWarnThreshold_const))))
          (Variables|VFX|SetTH_LifeSpanWarningThreshold _returnvalue)
          (CallFunction|StartStuckProjectileCheck))))
    (:then_1
      ;;[0.4132 0.02567 1] Sequence [1]: Setup Client FX and feedback
      (Networking|SwitchHasAuthority
        (:Remote
          (CallFunction|AVIndicatorTraversal)
          (bind _returnvalue_2 (Utilities|Time|SetTimerbyEvent (EventDispatchers|CreateEvent :!event Custom|AVIndicator_Traversal) 0.2 true))
          (Variables|VFX|SetTH_AVIndicatorRefresh _returnvalue_2))))
    (:then_2
      ;;[1 0 0.9249] Sequence [2]: Turn off overlap collision on the client,
      ;; While keeping the collision would allow for predictive stopping the projectile when it hits an intangible building, in those cases the projectile will be obscured by the building so predictive stopping isn't visually necessary.
      (Networking|SwitchHasAuthority
        (:Remote
          (Collision|SetCollisionEnabled (Variables|Default|GetSphere%20Component%20Overlap%20Buildings) "NoCollision"))))
    (:then_3
      ;;[0.1354 1 0.1527] Sequence [3]: Initialize InitialForwardVector on the server. This value is used to determine if a player decided to destroy the building they hit.
      (Networking|SwitchHasAuthority
        (:Authority
          (bind _asfort_player_controller (Utilities|Casting|CastToFortPlayerController (Game|GetInstigatorController))
            (:then
              (Variables|Traversal|SetInitialForwardVector (Math|Vector|GetForwardVector (Pawn|GetControlRotation _asfort_player_controller :!class /Script/Engine.Controller))))
            (:CastFailed
              (Variables|Traversal|SetInitialForwardVector (Math|Vector|Normalize (Variables|/Script/Engine.MovementComponent|GetVelocity (Variables|Movement|GetProjectileMovementComponent)))))))))
    (:then_4
      ;;[1 0 0.8581] Initialize The Current Desired Velocity. This solves an issue where on the client the velocity gets set to 0,0,z which causes the projectile rotation to be wrong for a frame.
      (Variables|Cached|SetCurrentDesiredVelocity%20XY (Math|Conversions|ToVector2D(Vector) (Variables|/Script/Engine.MovementComponent|GetVelocity (Variables|Movement|GetProjectileMovementComponent)))))))

(event EventEndPlay
  (params (EndPlayReason :type EEndPlayReason%20Enum :default "Destroyed"))
  ;[1 0.5394 0] EventEndPlay
  ; Cleanup BP: Unbind Events and Clear Timers
  (|CallParentFunction EndPlayReason)
  (Utilities|FlowControl|Sequence
    (:then_0
      ;;[1 0.5394 0] Sequence 0: Unbind Events
      (bind _buildingstoppedat (Variables|Cached|GetBuildingStoppedAt)
        (:then
          (EventDispatchers|UnbindEventfromOnDied _buildingstoppedat (EventDispatchers|CreateEvent :!event Custom|OnBuildingStoppedAtDied)))
        (:else)))
    (:then_1
      ;;[1 0.5394 0] Sequence 1: Clear Timers
      (Utilities|Time|ClearandInvalidateTimerbyHandle (Variables|StateTrackers|GetDealDamageTimer))
      (Utilities|Time|ClearandInvalidateTimerbyHandle (Variables|VFX|GetTH_TraversalAirborneTimer))
      (Utilities|Time|ClearandInvalidateTimerbyHandle (Variables|VFX|GetTH_LifeSpanWarningThreshold))
      (Utilities|Time|ClearandInvalidateTimerbyHandle (Variables|VFX|GetTH_AVIndicatorRefresh))
      (Utilities|Time|ClearandInvalidateTimerbyHandle (Variables|StuckCheck|GetTH_StuckBladeTimer))
      (Utilities|Time|ClearandInvalidateTimerbyHandle (Variables|DamagingandHit|GetTH_CheckBuildingStopped)))))

(event EventTick
  (params (DeltaSeconds :type Float%20%28single-precision%29))
  (bind _returnvalue (Game|GetInstigator))
  (bind _pawnlaunchdata (Variables|Knockback|GetPawnLaunchData))
  ;[0.1235 0.9502 1] Tick
  ; Update Velocity and accelerate the projectile
  (Utilities|FlowControl|Sequence
    (:then_0
      ;;[0.01025 0.01591 1] Launch all queued pawns and clear the array.
      (Networking|SwitchHasAuthority
        (:Authority
          (for _array_element _pawnlaunchdata
            (Utilities|IsValid (Utilities|Struct|BreakSTRUCTSawbladeLaunchCharacter _array_element)
              (:Is%20Valid
                (if (Variables|Default|GetLogLaunchingPawns)
                  (Fortnite|FortShippingLog self (Utilities|String|Append "Prj_Athena_Sawblade::KnockbackHitActor(): launching pawn: [" (Fortnite|GetPlayerorObjectNameSafe (Utilities|Struct|BreakSTRUCTSawbladeLaunchCharacter _array_element)) "] launch velocity: [" (Utilities|String|ToString(Vector) (.LaunchVelocity_5_1F593C5240E24DDA901851B90B3C697E (Utilities|Struct|BreakSTRUCTSawbladeLaunchCharacter _array_element))) "]."))
                  (else
                    _))
                (Character|LaunchCharacter (Utilities|Struct|BreakSTRUCTSawbladeLaunchCharacter _array_element) (.LaunchVelocity_5_1F593C5240E24DDA901851B90B3C697E (Utilities|Struct|BreakSTRUCTSawbladeLaunchCharacter _array_element)) false true)
                (bind _asfort_player_pawn_athena (Utilities|Casting|CastToFortPlayerPawnAthena (Utilities|Struct|BreakSTRUCTSawbladeLaunchCharacter _array_element)))
                (Class|FortPlayerPawnAthena|SetFallInstigator _asfort_player_pawn_athena (Game|GetInstigatorController) (Variables|Knockback|GetFallingDamageTags%20Const)))))
          (Utilities|Array|Clear _pawnlaunchdata))))
    (:then_1
      (Traversal|TickVelocity))
    (:then_2
      ;;[0.5742 0.2879 1] Turn on Instigator Collision oonce the projectile is far enough from the instigator pawn or after a specified amount of time (this way we try to avoid turning on collision during an overlap, but eventually we turn on collision).
      ;; We need to do this to allow the projectile to hit its instigator and for the instigator to hit the projectile with a pickax.
      (if (Variables|Collision|GetbIgnoreInstigatorCollision)
        (if (not (>= (Actor|GetGameTimeSinceCreation) (Variables|DamagingAndHit|GetStopIgnoringInstigatorFailsafeTimeConst)))
          (Utilities|IsValid _returnvalue
            (:Is%20Valid
              (if (>= (Math|Vector|DistanceSquared(Vector) (Transformation|GetActorLocation _returnvalue) (Transformation|GetActorLocation)) (Variables|DamagingAndHit|GetMinStopIgnoringInstigatorDistanceSquared%20Const))
                _))))
        (Collision|SetIgnoreInstigatorCollision false)))))

(event Custom|ActivateTraversalVFX
  (params (HitInfo :type Hit%20Result%20Structure))
  (bind (_bblockinghit _binitialoverlap _time _distance _location _impactpoint _normal _impactnormal _physmat) (Collision|BreakHitResult HitInfo))
  (bind _surfacetype (Variables|/Script/PhysicsCore.PhysicalMaterial|GetSurfaceType _physmat))
  (bind _vfx_niagra_traversal (Variables|Default|GetVFX_Niagra_Traversal))
  ;[0.4132 0.02567 1] VFX/SFX
  ;;[0.4132 0.02567 1] VFX: Play traversal effect for all non-destructive hits -
  ;; pass in Physical Material to allow for different effects based on what it is hitting, but only if the surface has changed
  (Utilities|IsValid _physmat
    (:Is%20Valid
      (if (Utilities|Enum|NotEqual(Enum) _surfacetype "SurfaceType8")
        (bind _returnvalue (Utilities|Time|SetTimerbyEvent (EventDispatchers|CreateEvent :!event Custom|StartAirborneVFX) (ScalableFloat|GetValueatLevel (Variables|VFX|GetHF_ConsiderAirborneTimer)))) ; set or reset timer to determine if the projectile is airborne
        (Variables|VFX|SetTH_TraversalAirborneTimer _returnvalue)
        (else
          (Utilities|Time|ClearandInvalidateTimerbyHandle (Variables|VFX|GetTH_TraversalAirborneTimer)) ; clear any existing timer
        ))
      (if (Utilities|Enum|NotEqual(Enum) (Variables|VFX|GetLastHitSurface) _surfacetype)
        (Variables|VFX|SetLastHitSurface _surfacetype)
        (Components|Activation|Activate _vfx_niagra_traversal false)))
    (:Is%20Not%20Valid
      _))
  (Effects|Components|ParticleSystem|SetIntParameter _vfx_niagra_traversal "PhysicalSurface" (DamagingAndHit|GetMaterialSurfaceIndex HitInfo))
  (Effects|Components|ParticleSystem|SetBoolParameter _vfx_niagra_traversal "bIsAirborne" false)
  (bind _physicalsurfaceindex (SFX|SetDamageAudioSurfaceType (Variables|Default|GetSFX_ImpactDamage) HitInfo))
  (bind _returnvalue_2 (Audio|SpawnSoundAttached "/SawbladeGun/Items/SawbladeGun/Sounds/SawBladeGun_Proj_Impact_Mat_Change_Cue.SawBladeGun_Proj_Impact_Mat_Change_Cue" (Variables|Transformation|GetRootComponent)))
  (Audio|Parameter|SetIntegerParameter _returnvalue_2 "PhysicalSurface" _physicalsurfaceindex :!class /Script/AudioExtensions.AudioParameterControllerInterface)
  (Utilities|FlowControl|DoOnce false
    (:Completed
      (Audio|SpawnSoundAttached (Variables|SFX|GetSFX_FirstImpact) (Variables|Default|GetStaticMesh)))))

(event Custom|Init
  (params (ChargePercent :type Float%20%28double-precision%29) (OrthogonalSpeed :type Float%20%28double-precision%29) (OrthogonalDirection :type Vector))
  ;[0.1083 1 0.0636] Init (called from weapon bp)
  (Variables|StateTrackers|SetCharge ChargePercent)
  (Variables|Cached|SetDesired%20Orthogonal%20Speed OrthogonalSpeed)
  (Variables|Cached|SetDesired%20Orthogonal%20Direction OrthogonalDirection)
  (Traversal|ConstraintoVerticalPlane)
  ;;[0.02879 1 0.1659] If we ever hit something and we only now set a valid initial direction then init velocities.
  (if (Variables|StateTrackers|GetHasEverHitSomething)
    (Traversal|OnHitSetInitialVelocities)
    (else
      _)))

(event Custom|StartAirborneVFX
  ;[0.4132 0.02567 1] VFX/SFX
  ;;[0.4132 0.02567 1] EndTraversalVFX
  ;; Remove the Traversal VFX from the Sawblade projectile and clear the timer; this Event is activated from a timer when the projectile has not hit a surface for a set amount of time which means it is considered airborne
  (Effects|Components|ParticleSystem|SetBoolParameter (Variables|Default|GetVFX_Niagra_Traversal) "bIsAirborne" true)
  (Utilities|Time|ClearandInvalidateTimerbyHandle (Variables|VFX|GetTH_TraversalAirborneTimer))
  (Audio|Components|Audio|FadeOut (Variables|Default|GetSFX_ImpactDamage) 0.2 0.0 "Linear" :!class /Script/Engine.AudioComponent)
  (Variables|SFX|SetbDamageAudioFaded true)
  (Variables|SFX|SetbAudioDeathControl false))

(event OnComponentHit(CapsuleComponent)
  (params (HitComponent :type Primitive%20Component%20Object%20Reference) (OtherActor :type Actor%20Object%20Reference) (OtherComp :type Primitive%20Component%20Object%20Reference) (NormalImpulse :type Vector :default "0, 0, 0") (Hit :type Hit%20Result%20Structure%20%28by%20ref%29))
  (bind (_bblockinghit _binitialoverlap _time _distance _location _impactpoint _normal _impactnormal _physmat _hitactor) (Collision|BreakHitResult Hit))
  ;[0.09237 0.1206 1] OnComponentHit (CapsuleComponent)
  ;;[1 0.8461 0.2999] Confirm that either A) we are allowed to handle when objects hit us OR
  ;; B) we hit the other actor
  (if (or (not (Variables|DamagingAndHit|GetbShouldSawbladeIgnoreActorsHittingIt)) (!= _hitactor self))
    (Variables|Cached|SetOtherActorHit OtherActor)
    (Variables|Cached|SetLastHitInfo Hit)
    (CallFunction|ActivateTraversalVFX Hit)
    (CallFunction|DetermineProjectileHitBehavior)))

(event OnEnterWaterEvent(WaterInteractionComponent)
  (params (WaterBody :type Fort%20Water%20Body%20Actor%20Object%20Reference) (WaterInteractionComponent :type Fort%20Water%20Interaction%20Component%20Object%20Reference) (bIsFirstBody :type Boolean))
  ;[0.09237 0.1206 1] OnComponentHit (CapsuleComponent)
  (CallFunction|ActivateTraversalVFX (Collision|MakeHitResult false false 0.0 0.0 "0, 0, 0" "0, 0, 0" "0,0,1" "0,0,1" (Variables|VFX|GetWaterPhysicalMaterial)))
  (if (not (Variables|StateTrackers|GetHasEverHitSomething))
    (Traversal|OnHitSetInitialVelocities)))

(event Custom|EventOnDestroyed
  (params (DestroyedActor :type Actor%20Object%20Reference))
  (bind _capsulecomponent (Variables|Collision|GetCapsuleComponent))
  (bind _returnvalue (Transformation|GetWorldLocation _capsulecomponent))
  ;[0.4132 0.02567 1] VFX/SFX
  ;;[0.4132 0.02567 1] EventOnDestroyed
  ;; Excute the Sawblade Gun destroyed Gameplay Cue for FX
  (Effects|Components|ParticleSystem|SpawnFXAtLocation (Variables|VFX|GetNS_Explode) _returnvalue)
  (Audio|PlaySoundatLocation (Variables|SFX|GetSFX_Destroyed) _returnvalue)
  (CallFunction|StartAirborneVFX)
  (Audio|Components|Audio|FadeOut (Variables|Default|GetSFX_Traversal) 0.5 0.0 "Linear" :!class /Script/Engine.AudioComponent)
  (Variables|SFX|SetbAudioDeathControl false))

(event Custom|Rep_ActivateDamagingVFX
  (bind (_value _returnvalue) (Utilities|Map|Find (Variables|SFX|GetDamageLoopAudioMap) (.SurfaceType (DamagingAndHit|GetMaterialSurfaceIndex (Variables|Cached|GetLastHitInfo)))))
  (bind _lasthitinfo (Variables|Cached|GetLastHitInfo))
  (bind (_surfaceindex _surfacetype) (DamagingAndHit|GetMaterialSurfaceIndex _lasthitinfo))
  (bind _vfx_niagra_damage (Variables|Default|GetVFX_Niagra_Damage))
  ;[0.4132 0.02567 1] VFX/SFX
  ;;[0.4132 0.02567 1] VFX - start looping vfx while damaging an enviroment object
  (Utilities|FlowControl|DoOnce false
    (:Completed
      (Networking|CanExecuteCosmeticEvents
        (:True
          (Components|Activation|Activate _vfx_niagra_damage false)
          (Effects|Components|ParticleSystem|SetIntParameter _vfx_niagra_damage "PhysicalSurface" _surfaceindex)
          (bind _returnvalue_2 (Audio|SpawnSoundatLocation "/SawbladeGun/Items/SawbladeGun/Sounds/SawBladeGun_Proj_Impact_Cue.SawBladeGun_Proj_Impact_Cue" (Transformation|GetActorLocation)))
          (Audio|Parameter|SetIntegerParameter _returnvalue_2 "PhysicalSurface" (Math|Integer|SelectInt _value 0 _returnvalue) :!class /Script/AudioExtensions.AudioParameterControllerInterface))))))

(event Custom|ResetDamagingVFX
  (bind (_value _returnvalue) (Utilities|Map|Find (Variables|SFX|GetDamageLoopAudioMap) (.SurfaceType (DamagingAndHit|GetMaterialSurfaceIndex (Variables|Cached|GetLastHitInfo)))))
  (bind _lasthitinfo (Variables|Cached|GetLastHitInfo))
  (bind (_surfaceindex _surfacetype) (DamagingAndHit|GetMaterialSurfaceIndex _lasthitinfo))
  (bind _vfx_niagra_damage (Variables|Default|GetVFX_Niagra_Damage))
  ;[0.4132 0.02567 1] VFX/SFX
  ;;[0.4132 0.02567 1] VFX - start looping vfx while damaging an enviroment object
  (Utilities|FlowControl|DoOnce false
    (:Completed
      (Networking|CanExecuteCosmeticEvents
        (:True
          (Components|Activation|Activate _vfx_niagra_damage false)
          (Effects|Components|ParticleSystem|SetIntParameter _vfx_niagra_damage "PhysicalSurface" _surfaceindex)
          (bind _returnvalue_2 (Audio|SpawnSoundatLocation "/SawbladeGun/Items/SawbladeGun/Sounds/SawBladeGun_Proj_Impact_Cue.SawBladeGun_Proj_Impact_Cue" (Transformation|GetActorLocation)))
          (Audio|Parameter|SetIntegerParameter _returnvalue_2 "PhysicalSurface" (Math|Integer|SelectInt _value 0 _returnvalue) :!class /Script/AudioExtensions.AudioParameterControllerInterface))))))

(event Custom|DetermineProjectileHitBehavior
  ;[0.2958 0.447 0.6042] DetermineProjectileHitBehavior
  ; Setup velocity, check hit type and DOT product of angle to determine the hit type (stop and destroy, knock back and continue, or ignore)
  (Utilities|FlowControl|Sequence
    (:then_0
      (Networking|SwitchHasAuthority
        (:Authority
          (Utilities|FlowControl|Sequence
            (:then_1
              (if (not (Variables|StateTrackers|GetStopped))
                (bind _buildingactor (Utilities|ShouldStoporDamageLastHitBuilding)
                  (:Stop%20AND%20Damage%20Building
                    (if (Variables|/Script/Engine.Actor|GetbCanBeDamaged _buildingactor)
                      (CallFunction|StopProjectileandDamageBuildingActor _buildingactor)
                      ;;[0.2958 0.447 0.6042] Setup timer to check if the object stopped at should still be destroyed
                      ;; (hasn't moved, had collision disabled, player built edited, door opened, etc.)
                      (bind _returnvalue (Utilities|Time|SetTimerbyEvent (EventDispatchers|CreateEvent :!event Custom|ResumeMovementIfStoppedAndNotColliding) (Math|Float|SelectFloat (ScalableFloat|GetValueatLevel (Variables|DamagingAndHit|GetHF_ReevaluteStopCollisionTime)) 0.1 (> (ScalableFloat|GetValueatLevel (Variables|DamagingAndHit|GetHF_ReevaluteStopCollisionTime)) 0)) true))
                      (Variables|DamagingandHit|SetTH_CheckBuildingStopped _returnvalue)
                      (else
                        (Traversal|ReduceLifeSpanToDieAfterWarning))))
                  (:Damage%20Actor%20NO%20Stopping
                    _)))
              (CallFunction|TryKnockbackandDamageLastHitActor))
            (:then_2
              ;;[0.1874 1 0.9901] Only call OnHitSetInitialVelocities if this is the first hit detected by this projectile.
              ;; Only call OnHitSetInitialVelocities AFTER calling ShouldStopAndDamageBuilding (so ShouldStopAndDamageBuilding can look at HasEverHitSomething before it gets set to true.)
              (if (and (not (Variables|StateTrackers|GetHasEverHitSomething)) (Collision|BreakHitResult (Variables|Cached|GetLastHitInfo)))
                (Traversal|OnHitSetInitialVelocities) ; setup traversal velocity the \nfirst time it hits something
              ))))))
    (:then_1
      ;;[1 0.4584 0.09994] Clear Last Hit Data after we finish using them.
      (Variables|Cached|SetLastHitInfo)
      (Variables|Cached|SetOtherActorHit 0))))

(event Custom|LifeSpanNearEndWarning
  ;[0.4132 0.02567 1] VFX/SFX
  ;;[0.4132 0.02567 1] LifeSpanNearEndWarning
  ;; Client Only - visual animation of wobble and FX for when the sawblade is near its end lifetime
  (Utilities|Time|ClearandInvalidateTimerbyHandle (Variables|VFX|GetTH_LifeSpanWarningThreshold))
  (Networking|CanExecuteCosmeticEvents
    (:True
      (bind (_direction _wobblerotationx) (Timeline|EndLifeSpawn_Wobble PlayFromStart)
        (:Update
          (Transformation|SetRelativeRotation (Variables|Default|GetMeshRotationVertex) (Math|Rotator|MakeRotator _wobblerotationx (.pitch (Variables|/Script/Engine.SceneComponent|GetRelativeRotation (Variables|Default|GetMeshRotationVertex))) (.yaw (Variables|/Script/Engine.SceneComponent|GetRelativeRotation (Variables|Default|GetMeshRotationVertex)))) false false))))))

(event Custom|ResetLifeSpanWarning
  (params (LifespanChange :type Float%20%28double-precision%29))
  ;[0.4132 0.02567 1] VFX/SFX
  ;;[0.4132 0.02567 1] LifeSpanNearEndWarning
  ;; Client Only - visual animation of wobble and FX for when the sawblade is near its end lifetime
  (Utilities|Time|ClearandInvalidateTimerbyHandle (Variables|VFX|GetTH_LifeSpanWarningThreshold))
  (bind _returnvalue (Utilities|Time|SetTimerbyEvent (EventDispatchers|CreateEvent :!event Custom|LifeSpanNearEndWarning) (- LifespanChange (Variables|VFX|GetLifeSpanEndWarnThreshold_const))))
  (Variables|VFX|SetTH_LifeSpanWarningThreshold _returnvalue)
  (Networking|CanExecuteCosmeticEvents
    (:True
      (Transformation|SetRelativeRotation (Variables|Default|GetMeshRotationVertex) (0.0 0.0 0.0) false false) ; reset rotation if it was being animated
      (Timeline|EndLifeSpawn_Wobble Stop))))

(event Custom|AVIndicator_Traversal
  (bind _returnvalue (Game|GetPlayerPawn 0))
  ;[0.4132 0.02567 1] VFX/SFX
  ;;[0.4132 0.02567 1] AVIndicatorTraversal
  ;; Show the Audio visual indicator as an attack or vehicle depending on if the blade is hostile or friendly
  ;; Determine hostile from the local player and the blade team affiliation
  (Utilities|IsValid _returnvalue
    (:Is%20Valid
      (Fortnite|ShowSoundIndicatorLocal :TrackedActor self :Location (Transformation|GetActorLocation self) :MaxAudibleDistance (ScalableFloat|GetValueatLevel (Variables|VFX|GetHF_AVIndicator_Distance)) :IndicatorType (Utilities|Select :Index (Team|GetTeamAffiliation self _returnvalue) :Friendly "Vehicle" :Neutral "Vehicle" :Hostile "Gunshot") :Instigator self :Tint "(R=0.000000,G=0.000000,B=0.000000,A=0.000000)" :OverrideIcon "/SawbladeGun/Items/SawbladeGun/Icons/AV_Sawblade_32.AV_Sawblade_32" :AttenuationMaxRangeMultiplier 1.0 :AttenuationIntensityMultiplier 1.0))))

(event Custom|PickaxeImpactAudioMulticast
  ;[0.4132 0.02567 1] VFX/SFX
  ;;[0.4132 0.02567 1] Hitting the sawblade with a pickaxe
  (Networking|CanExecuteCosmeticEvents
    (:True
      (Audio|PlaySoundatLocation (Variables|SFX|GetSawBladePickaxeImpactSound) (Transformation|GetActorLocation)))))

(event OnComponentBeginOverlap(SphereComponentOverlapBuildings)
  (params (OverlappedComponent :type Primitive%20Component%20Object%20Reference) (OtherActor :type Actor%20Object%20Reference) (OtherComp :type Primitive%20Component%20Object%20Reference) (OtherBodyIndex :type Integer) (bFromSweep :type Boolean) (SweepResult :type Hit%20Result%20Structure%20%28by%20ref%29))
  ;[0.4744 0.4583 1] OnComponentBeginOverlap and OnComponentEndOverlap (Sphere Component Detect Overlap Buildings and Sphere Component Overlap Pawns)
  ; When overlapping with an actor try to damage them and add them to a list of overlapping actors to attempt to damage frequently.
  ; If an actor leaves the overlap, remove that actor and their timer from the list of overlapping actors.
  (if (not (Components|ComponentHasTag OtherComp "IgnoreSawBlade"))
    (Networking|SwitchHasAuthority
      (:Authority
        (DamagingAndHit|DetermineProjectileOverlapBehaviorforActor OtherActor)))))

(event OnComponentEndOverlap(SphereComponentOverlapBuildings)
  (params (OverlappedComponent :type Primitive%20Component%20Object%20Reference) (OtherActor :type Actor%20Object%20Reference) (OtherComp :type Primitive%20Component%20Object%20Reference) (OtherBodyIndex :type Integer))
  ;[0.4744 0.4583 1] OnComponentBeginOverlap and OnComponentEndOverlap (Sphere Component Detect Overlap Buildings and Sphere Component Overlap Pawns)
  ; When overlapping with an actor try to damage them and add them to a list of overlapping actors to attempt to damage frequently.
  ; If an actor leaves the overlap, remove that actor and their timer from the list of overlapping actors.
  (DamagingAndHit|EndOverlap OtherActor))

(event OnComponentBeginOverlap(SphereComponentOverlapPawns)
  (params (OverlappedComponent :type Primitive%20Component%20Object%20Reference) (OtherActor :type Actor%20Object%20Reference) (OtherComp :type Primitive%20Component%20Object%20Reference) (OtherBodyIndex :type Integer) (bFromSweep :type Boolean) (SweepResult :type Hit%20Result%20Structure%20%28by%20ref%29))
  ;[0.4744 0.4583 1] OnComponentBeginOverlap and OnComponentEndOverlap (Sphere Component Detect Overlap Buildings and Sphere Component Overlap Pawns)
  ; When overlapping with an actor try to damage them and add them to a list of overlapping actors to attempt to damage frequently.
  ; If an actor leaves the overlap, remove that actor and their timer from the list of overlapping actors.
  (Networking|SwitchHasAuthority
    (:Authority
      ;;[1 0.9493 0] Make sure we do NOT process overlapping with our instigator if we are ignoring instigator collision.
      (if (or (!= OtherActor (Variables|Actor|GetInstigator)) (not (Variables|Collision|GetbIgnoreInstigatorCollision)))
        (DamagingAndHit|DetermineProjectileOverlapBehaviorforActor 0)))))

(event OnComponentEndOverlap(SphereComponentOverlapPawns)
  (params (OverlappedComponent :type Primitive%20Component%20Object%20Reference) (OtherActor :type Actor%20Object%20Reference) (OtherComp :type Primitive%20Component%20Object%20Reference) (OtherBodyIndex :type Integer))
  ;[0.4744 0.4583 1] OnComponentBeginOverlap and OnComponentEndOverlap (Sphere Component Detect Overlap Buildings and Sphere Component Overlap Pawns)
  ; When overlapping with an actor try to damage them and add them to a list of overlapping actors to attempt to damage frequently.
  ; If an actor leaves the overlap, remove that actor and their timer from the list of overlapping actors.
  (DamagingAndHit|EndOverlap OtherActor))

(event Custom|KnockbackOverlappingActorTimerExpired
  (bind _knockbacktimerhandlesbyoverlappingvictimactormap (Variables|Knockback|GetKnockbackTimerHandlesByOverlappingVictimActorMap))
  ;[0 1 0.9908] KnockbackOverlappingActorTimerExpired
  ; Check if the timer expired for any actors in KnockbackTimerHandlesByOverlappingVictimActorMap.
  ; If a timer has expired attempt to knockback/damage that actor.
  (bind _keys (Utilities|Map|Keys _knockbacktimerhandlesbyoverlappingvictimactormap))
  (bind _values (Utilities|Map|Values _knockbacktimerhandlesbyoverlappingvictimactormap))
  (for (_array_element _array_index) _values
    (if (<= (Utilities|Time|GetTimerRemainingTimebyHandle _array_element) 0)
      (DamagingAndHit|DetermineProjectileOverlapBehaviorforActor (Utilities|Array|Get(acopy) _keys _array_index)))))

(event Custom|TryKnockbackAndDamageLastHitActor
  (bind _knockbacktimesbyvictimactormap (Variables|Knockback|GetKnockbackTimesByVictimActorMap))
  (bind (_value _returnvalue) (Utilities|Map|Find _knockbacktimesbyvictimactormap (Variables|Cached|GetOtherActorHit)))
  (bind _returnvalue_2 (Fortnite|Sync|Time|GetServerWorldTimeSeconds))
  (bind _otheractorhit (Variables|Cached|GetOtherActorHit))
  ;[0.401 0.2071 0.04981] TryKnockbackAndDamageLastHitActor
  ; Launch characters and vehicles if we run into them (and haven't knocked them back recently)
  (Networking|SwitchHasAuthority
    (:Authority
      ;;[0.451 0.4177 1] Filter out all unsupported knockback actors
      (Utilities|CasttoKnockbackSupportingActor _otheractorhit
        (:Fort%20Pawn
          _)
        (:Fort%20Vehicle%20Interface
          _)
        (:Knockbackable%20Building%20Actor
          _))
      (if (or (>= (- _returnvalue_2 _value) (Variables|Knockback|GetKnockbackActorRefreshTime%20Const)) (not _returnvalue))
        (DamagingAndHit|RemoveStaleKnockbackActors)
        (Utilities|Map|Add _knockbacktimesbyvictimactormap _otheractorhit _returnvalue_2)
        (DamagingAndHit|KnockbackHitActor (Variables|Cached|GetLastHitInfo) _otheractorhit)))))

(event Custom|StopProjectileAndDamageBuildingActor
  (params (HitBuildingActor :type Building%20Actor%20Object%20Reference))
  ;[0.401 0.2071 0.04981] StopProjectileAndDamageBuildingActor
  (Networking|SwitchHasAuthority
    (:Authority
      (if (and (Utilities|IsValid HitBuildingActor :!class /Script/Engine.KismetSystemLibrary) (!= (Variables|Cached|GetBuildingStoppedAt) HitBuildingActor))
        (bind _buildingstoppedat (Variables|Cached|GetBuildingStoppedAt)
          (:then
            (EventDispatchers|UnbindEventfromOnDied _buildingstoppedat (EventDispatchers|CreateEvent :!event Custom|OnBuildingStoppedAtDied)))
          (:else
            _))
        (Variables|Cached|SetBuildingStoppedAt HitBuildingActor)
        (Variables|StateTrackers|SetStopped true)
        (Variables|StateTrackers|SetAtMaxSpeed false)
        (bind _returnvalue (Utilities|Time|SetTimerbyEvent (EventDispatchers|CreateEvent :!event Custom|DealDamageToStoppedAtBuilding) (ScalableFloat|GetValueatLevel (Variables|DamagingAndHit|Gethf_DamageRate)) true))
        (Variables|StateTrackers|SetDealDamageTimer _returnvalue)
        (EventDispatchers|BindEventtoOnDied (Variables|Cached|GetBuildingStoppedAt) (EventDispatchers|CreateEvent :!event Custom|OnBuildingStoppedAtDied))))))

(event Custom|ResumeMovementIfStoppedAndNotColliding
  ;[0 1 0.9908] CheckBuildingStoppedAtNotColliding
  (bind _stillhittingbuildingstoppedat (CallFunction|CheckStillHittingBuildingStoppedAt))
  (if (not _stillhittingbuildingstoppedat)
    (CallFunction|StartMovingAgainonServer)))

(event Custom|OnBuildingStoppedAtDied
  (params (DamagedActor :type Actor%20Object%20Reference) (Damage :type Float%20%28single-precision%29) (InstigatedBy :type Controller%20Object%20Reference) (DamageCauser :type Actor%20Object%20Reference) (HitLocation :type Vector) (FHitComponent :type Primitive%20Component%20Object%20Reference) (BoneName :type Name) (Momentum :type Vector))
  ;[0.401 0.2071 0.04981] OnBuildingStoppedAtDied
  (CallFunction|StartMovingAgainonServer)
  ;[1 1 1] Quest Stack Tracking
  (bind _result (Math|Integer|IncrementInt (Variables|Default|GetQuest%20Stack%20Tracking))
    (:%20%20
      (if (>= _result 5)
        (Quest|SendComplexCustomStatEvent (Quests|GetQuestManager (Fortnite|GetFortPlayerControllerfromActor InstigatedBy) "Athena") 0 "(GameplayTags=((TagName=\"Athena.Quests.SawbladeDestroyX\")))" "(GameplayTags=)" 1)))))

(event Custom|StartMovingAgainOnServer
  ;[0.04898 1 0] StartMovingAgainOnServer
  (Networking|SwitchHasAuthority
    (:Authority
      (if (Variables|StateTrackers|GetStopped)
        (Variables|StateTrackers|SetStopped false)
        (Utilities|Time|ClearandInvalidateTimerbyHandle (Variables|DamagingandHit|GetTH_CheckBuildingStopped))
        (bind _buildingstoppedat (Variables|Cached|GetBuildingStoppedAt)
          (:then
            (EventDispatchers|UnbindEventfromOnDied _buildingstoppedat (EventDispatchers|CreateEvent :!event Custom|OnBuildingStoppedAtDied))
            (Variables|Cached|SetBuildingStoppedAt 0))
          (:else))))))

(event Custom|DealDamageToStoppedAtBuilding
  (bind _buildingstoppedat (Variables|Cached|GetBuildingStoppedAt))
  ;[0.1235 0.9502 1] DealDamageToStoppedAtBuilding() calls DealDamageToBuilding(BuildingStoppedAt)
  ; DealDamageToBuilding() deals damage over time by applying GE
  (Utilities|IsValid _buildingstoppedat
    (:Is%20Valid
      (CallFunction|DealDamagetoBuilding _buildingstoppedat true))
    (:Is%20Not%20Valid
      (CallFunction|StartMovingAgainonServer))) ; If we have a buiding we are stopped at deal damage to it.\nElse the building is missing start moving again.
)

(event Custom|DealDamageToBuilding
  (params (BuildingToDamage :type Building%20Actor%20Object%20Reference) (AddImpulse :type Boolean))
  (bind _returnvalue (Game|GetInstigator BuildingToDamage))
  ;[0.1235 0.9502 1] DealDamageToStoppedAtBuilding() calls DealDamageToBuilding(BuildingStoppedAt)
  ; DealDamageToBuilding() deals damage over time by applying GE
  (Networking|SwitchHasAuthority
    (:Authority
      (Utilities|IsValid BuildingToDamage
        (:Is%20Valid
          (Utilities|IsValid _returnvalue
            (:Is%20Valid
              ;;[0 1 0.6416 0.5] If this building is attached to a pawn (that is NOT our instigator) then listen for blocking events, deal damage, and then remove the listener.
              (bind (_asynctaskproxy _payload) (Ability|Async|WaitGameplayEventToActor _returnvalue (Variables|DamagingAndHit|GetProjectile%20Blocked%20Tag%20Const) true)
                (:then
                  ;;;[0 0.9478 1 0.3] Trigger damage against the last hit building and immediately afterward STOP listening for the Gameplay Event
                  (DamagingAndHit|DealDamageToBuildingFunction BuildingToDamage AddImpulse)
                  (Ability|Async|EndAction _asynctaskproxy))
                (:EventReceived
                  ;;;[0.3594 0.219 0.08449] The instigator we were targetting for damage blocked while we were dealing damage to our target (i.e. our target blocked our damage). So now we want to "reflect" the projectile (i.e. give our projectile a new heading and instigator.)
                  (Utilities|IsValid (Fortnite|GetFortPlayerControllerfromActor (.Instigator (Utilities|Struct|BreakGameplayEventData _payload)))
                    (:Is%20Valid
                      (CallFunction|SetNewHeading (Pawn|GetControlRotation (Fortnite|GetFortPlayerControllerfromActor (.Instigator (Utilities|Struct|BreakGameplayEventData _payload))) :!class /Script/Engine.Controller) (Fortnite|GetFortPlayerControllerfromActor (.Instigator (Utilities|Struct|BreakGameplayEventData _payload)))))
                    (:Is%20Not%20Valid
                      (Utilities|IsValid (AI|GetAIController (.Instigator (Utilities|Struct|BreakGameplayEventData _payload)))
                        (:Is%20Valid
                          (CallFunction|SetNewHeading (Pawn|GetControlRotation (AI|GetAIController (.Instigator (Utilities|Struct|BreakGameplayEventData _payload))) :!class /Script/Engine.Controller) (AI|GetAIController (.Instigator (Utilities|Struct|BreakGameplayEventData _payload)))))) ; Try with AI
                    )))))
            (:Is%20Not%20Valid
              _)))
        (:Is%20Not%20Valid
          _))
      (DamagingAndHit|DealDamageToBuildingFunction BuildingToDamage AddImpulse))))

(event ImpulseChangedEvent(FortImpulseHandler)
  (params (ImpulsePoint :type Vector%20%28by%20ref%29 :default "0, 0, 0") (ImpulseMagnitude :type Vector%20%28by%20ref%29 :default "0, 0, 0") (TagContainer :type Gameplay%20Tag%20Container%20Structure%20%28by%20ref%29) (EffectContext :type Gameplay%20Effect%20Context%20Handle%20Structure%20%28by%20ref%29) (Instigator :type Controller%20Object%20Reference) (EffectCauser :type Actor%20Object%20Reference))
  ;[0.2188 0.1276 0] When impulse is added to this object, if the damage is from a pickax with a valid pawn, THEN redirect this projectile.
  (Utilities|Casting|CastToFortWeaponPickaxeAthena EffectCauser)
  (Utilities|IsValid EffectCauser
    (:Is%20Valid
      (CallFunction|SetNewHeading (Pawn|GetControlRotation Instigator :!class /Script/Engine.Controller) Instigator)
      (CallFunction|PickaxeImpactAudioMulticast))))

(event Custom|SetNewHeading
  (params (New%20Heading%20Rotation :type Rotator) (New%20Instigator%20Controller :type Controller%20Object%20Reference))
  (bind _hf_lifespantorestoreonhit (Variables|Traversal|Gethf_LifespanToRestoreOnHit))
  (bind _returnvalue (ScalableFloat|GetValueatLevel _hf_lifespantorestoreonhit))
  (bind _returnvalue_2 (Math|Vector|GetForwardVector New%20Heading%20Rotation))
  (bind _returnvalue_3 (Math|Conversions|ToVector2D(Vector) _returnvalue_2))
  (bind _returnvalue_4 (Pawn|GetControlledPawn New%20Instigator%20Controller))
  (bind _reflectionscounter (Variables|DamagingAndHit|GetReflectionsCounter))
  ;[0.6867 0.6771 1] SetNewHeading Called to redirect the projectile direction and to change the projectile's instigator.
  (Networking|SwitchHasAuthority
    (:Authority
      (Math|Float|IncrementFloat _reflectionscounter
        (:%20%20
          (if (>= _reflectionscounter (ScalableFloat|GetValueatLevel (Variables|DamagingAndHit|GetHF_LifetimeMaxReflections)))
            ;;[1 0.381 0] Destroy Actor
            (Actor|DestroyActor)
            (else
              (Utilities|FlowControl|Sequence
                (:then_0
                  ;;[1 0.9213 0.03484] Log
                  (Fortnite|FortShippingLog self (Utilities|String|Append "Prj_Athena_Sawblade::SetNewHeading() New Instigator: [" (Fortnite|GetPlayerNameSafe New%20Instigator%20Controller) "] New Heading Yaw: [" (Utilities|String|ToString(Rotator) New%20Heading%20Rotation) "].")))
                (:then_1
                  ;;[0.1832 1 0.9203] Set the new direction of this projectile
                  (if (not (Math|Vector2D|IsNearlyZero2D _returnvalue_3))
                    (bind _output_get (Variables|Cached|SetDesired%20Orthogonal%20Direction (Math|Conversions|ToVector(Vector2D) (Math|Vector2D|Normalize2D _returnvalue_3))))
                    (Projectile|SetVelocity (Variables|Movement|GetProjectileMovementComponent) (* _output_get (Math|Vector|VectorLengthXY (Variables|/Script/Engine.MovementComponent|GetVelocity (Variables|Movement|GetProjectileMovementComponent)))))
                    (Traversal|OnHitSetInitialVelocities)
                    (Transformation|SetWorldRotation (Variables|Collision|GetCapsuleComponent) (Math|Rotator|MakeRotfromX _output_get) false false) ; Rotate the root so that the visuals face the correct direction.
                    (CallFunction|StartMovingAgainonServer)))
                (:then_2
                  ;;[0.1692 0.147 1] Set the inew instigator of this projectile
                  (Utilities|IsValid New%20Instigator%20Controller
                    (:Is%20Valid
                      (Utilities|IsValid _returnvalue_4
                        (:Is%20Valid
                          (Variables|/Script/Engine.Actor|SetInstigator _returnvalue_4 self))))))
                (:then_3
                  ;;[0 0.0563 1] Change the lifespawn to allow for the projectile to live longer if the trajectory was adjusted to reward players that can pull that off
                  (if (< (Actor|GetLifeSpan) _returnvalue)
                    (Actor|SetLifeSpan _returnvalue)
                    (CallFunction|ResetLifeSpanWarning _returnvalue)
                    (CallFunction|EndStuckProjectileCheck) ; reset stuck projectile check by ending and starting it again
                    (CallFunction|StartStuckProjectileCheck)))))))))))

(event Custom|StartStuckProjectileCheck
  ;[0 0.004631 0.1771 1 45] PROJECTILE STUCK CHECK
  ;;[0.04898 1 0] StartStuckProjectileCheck
  ;; Begins pollingtimer to check if a projectile is stuck on a non-building actor - called when spawned and when resuming movement
  (Variables|StuckCheck|SetLastStuckBladeCheckLocation (Transformation|GetActorLocation self))
  (bind _returnvalue (Utilities|Time|SetTimerbyEvent (EventDispatchers|CreateEvent :!event Custom|CheckMoveDistanceForStuckBlade) (ScalableFloat|GetValueatLevel (Variables|StuckCheck|Gethf_StuckCheckTime)) true))
  (Variables|StuckCheck|SetTH_StuckBladeTimer _returnvalue))

(event Custom|CheckMoveDistanceForStuckBlade
  (bind _self self)
  (bind _returnvalue (Transformation|GetActorLocation _self))
  ;[0 0.004631 0.1771 1 45] PROJECTILE STUCK CHECK
  ;;[0.1235 0.9502 1] CheckMoveDistanceForStuckBlade
  ;; Compares stored position and current position against min movement distance threshold to consider the blade stuck, then stores position for future check or reduces lifespan to die
  (if (< (Math|Float|Square (ScalableFloat|GetValueatLevel (Variables|StuckCheck|Gethf_StuckCheckDistance))) (Math|Vector|DistanceSquared(Vector) (Variables|StuckCheck|GetLastStuckBladeCheckLocation) _returnvalue))
    (Variables|StuckCheck|SetLastStuckBladeCheckLocation _returnvalue) ; store current position for next check
    (else
      (Traversal|ReduceLifeSpanToDieAfterWarning))))

(event Custom|EndStuckProjectileCheck
  ;[0 0.004631 0.1771 1 45] PROJECTILE STUCK CHECK
  ;;[1 0.5394 0] EndStuckProjectileCheck
  ;; Clear the polling timer - called when the projectile is intentionally stopped to destroy building actor
  (Utilities|Time|ClearandInvalidateTimerbyHandle (Variables|StuckCheck|GetTH_StuckBladeTimer)))

(fn ApplyImpactDamage
  (params (HitResult :type Hit%20Result%20Structure) (HitObject :type Object%20Reference))
  (bind _instigator (Variables|Actor|GetInstigator))
  (bind _returnvalue (Ability|GetFortAbilitySystemComponent _instigator))
  (bind _asactor (Utilities|Casting|CastToActor HitObject))
  (Utilities|IsValid _instigator
    (:Is%20Valid
      (bind _returnvalue_2 (CallFunction|GetOwnerWeapon))
      (bind _output_get (LocalVariables|SetEffectContext (GameplayEffects|MakeFortEffectContext _returnvalue _returnvalue_2 _returnvalue_2 _returnvalue_2 0)))
      (Ability|EffectContext|AddHitResult _output_get HitResult) ; We need to add a hit result in order for damage numbers to appear when this projectile hits vehicles.\nThis happens in AFortPawn::Call_InvokeGameplayCueExecuted_FromSpec() where AccumulatedBatchData_Shared.bIsValid is set to true;
      (GameplayEffects|ApplyGameplayEffectToTarget _returnvalue "/SawbladeGun/Items/SawbladeGun/GE_SawbladeGun_ApplyDamageImpact.GE_SawbladeGun_ApplyDamageImpact_C" (Ability|GetFortAbilitySystemComponent _asactor) 0.0 _output_get)
      (bind _asfort_pawn (Utilities|Casting|CastToFortPawn _asactor))
      (Utilities|FlowControl|Sequence
        (:then_0
          ;[0.4132 0.02567 1] Sequence 0: Hit notification for eliminated targets
          (if (FortPawn|IsDead _asfort_pawn)
            (bind _returnvalue_3 (CallFunction|GetOwnerWeapon))
            (bind _asb_sawblade_gun_athena (Utilities|Casting|CastToB_SawbladeGun_Athena _returnvalue_3))
            (if (not (Variables|SFX|GetbAudioDeathControl))
              (Class|BSawbladeGunAthena|Play2DHitNotification _asb_sawblade_gun_athena)
              (Variables|SFX|SetbAudioDeathControl true))))
        (:then_1
          ;[0.4167 0.2072 0.05745] Sequence 1: Destroy the projectile if it hits certain targets - often enemies that wouldn't get bounced and may block the blades traversal (ex. The Caretaker)
          (if (GameplayTags|HasAnyMatchingGameplayTags _asfort_pawn (Variables|Default|GetDestroyProjectileOnHitTags))
            (Actor|DestroyActor)))))))

(fn CheckStillHittingBuildingStoppedAt
  ;[0.2958 0.447 0.6042] CheckStillHittingBuildingStoppedAt
  ; Does a Capsule Trace over the current position forward with the last velocity to determine whether a hit is still occuring on the targetd building object (to determine whether the projectile should resume movement)
  (LocalVariables|SetHittingBuildingStoppedAt false) ; reset variable to return valid hit
  (bind _outhits (Collision|MultiCapsuleTraceByChannel (Projectile|GetCurrentLocation :!class /Script/FortniteGame.FortProjectileMovementInterface) (+ (Projectile|GetCurrentLocation :!class /Script/FortniteGame.FortProjectileMovementInterface) (* (Variables|Cached|GetPreStoppedVelocity) (Math|Float|SelectFloat (ScalableFloat|GetValueatLevel (Variables|DamagingAndHit|GetHF_ReevaluteStopCollisionTime)) 0.1 (> (ScalableFloat|GetValueatLevel (Variables|DamagingAndHit|GetHF_ReevaluteStopCollisionTime)) 0)))) (Variables|/Script/Engine.CapsuleComponent|GetCapsuleRadius (Variables|Collision|GetCapsuleComponent)) (Variables|/Script/Engine.CapsuleComponent|GetCapsuleHalfHeight (Variables|Collision|GetCapsuleComponent)) "TraceTypeQuery6")) ; Check the environmental collision of the projectile \nto verify the hit object is still present \n(moved, player build has been edited, door opened, etc.)
  (bind _array_element (Utilities|Array|ForEachLoopwithBreak _outhits)
    (:LoopBody
      (if (and (== (Variables|Cached|GetBuildingStoppedAt) (.HitActor (Collision|BreakHitResult _array_element))) (Collision|BreakHitResult _array_element))
        (LocalVariables|SetHittingBuildingStoppedAt true) ; store that the stopped at building is still hit
      ))
    (:Completed
      _))
  ;;[1 0.2603 0.2703] RETURN - BuildingStoppedAt still hit?
  (return (LocalVariables|GetHittingBuildingStoppedAt)))

(fn ConstrainToVerticalPlane
  ;[0.3906 0.3603 0.1261] Constrain this projectile to the plane perpendicular to Desired Orthogonal Direction and the Up Vector (an XZ plane).
  ; This ensures that even if the sawblade is traveling along a banked surface it will not veer left or right, it will only move forward + up/down.
  (Networking|SwitchHasAuthority
    (:Authority
      (Variables|Traversal|SetPlaneOriginConstraint%20Replicated (Transformation|GetWorldLocation (Variables|Transformation|GetRootComponent)))
      (Variables|Traversal|SetPlaneNormalConstraint%20Replicated (Math|Vector|CrossProduct (Variables|Cached|GetDesired%20Orthogonal%20Direction) "0, 0,1.000000")))))

(fn DealDamagetoBuildingFunction
  (params (BuildingToDamage :type Building%20Actor%20Object%20Reference) (AddImpulse :type Boolean))
  (bind _returnvalue (Ability|GetFortAbilitySystemComponent (Variables|Actor|GetInstigator)))
  (bind _returnvalue_2 (Ability|GetFortAbilitySystemComponent BuildingToDamage))
  (bind _instigator (Variables|Actor|GetInstigator))
  (Networking|SwitchHasAuthority
    (:Authority
      (Utilities|IsValid BuildingToDamage
        (:Is%20Valid
          (bind _returnvalue_3 (CallFunction|GetOwnerWeapon))
          ;[1 0.8177 0] Create a fake hit result so that harvested resources can be dropped at the correct location.
          (bind _output_get (LocalVariables|SetEffect%20Context (Class|AthenaFunctionLibrary|MakeSimpleFortEffectContext _instigator _returnvalue_3)))
          (Ability|EffectContext|AddHitResult _output_get (Collision|MakeHitResult false false 0.0 0.0 (Transformation|GetActorLocation) (Transformation|GetActorLocation) (Transformation|GetActorForwardVector) (* (Transformation|GetActorForwardVector) -1.0) 0 BuildingToDamage))
          (GameplayEffects|FortApplyGameplayEffectToTarget _returnvalue "/SawbladeGun/Items/SawbladeGun/GE_SawbladeGun_ApplyDamageHarvesting.GE_SawbladeGun_ApplyDamageHarvesting_C" _returnvalue_2 0 _instigator _returnvalue_3 0.0 _output_get)
          (if AddImpulse
            (GameplayEffects|FortApplyGameplayEffectToTarget _returnvalue "/Game/Abilities/Weapons/Ranged/GE_Impulse.GE_Impulse_C" _returnvalue_2 0 _instigator _returnvalue_3 0.0 _output_get)
            (else
              _))
          (CallFunction|RepActivateDamagingVFX))))))

(fn DetermineProjectileOverlapBehaviorForActor
  (params (Overlapping%20Actor :type Actor%20Object%20Reference))
  (bind _returnvalue (Fortnite|Sync|Time|GetServerWorldTimeSeconds))
  (bind _knockbacktimerhandlesbyoverlappingvictimactormap (Variables|Knockback|GetKnockbackTimerHandlesByOverlappingVictimActorMap))
  (bind (_value _returnvalue_2) (Utilities|Map|Find _knockbacktimerhandlesbyoverlappingvictimactormap Overlapping%20Actor))
  (bind _knockbacktimesbyvictimactormap (Variables|Knockback|GetKnockbackTimesByVictimActorMap))
  (bind (_value_2 _returnvalue_3) (Utilities|Map|Find _knockbacktimesbyvictimactormap Overlapping%20Actor))
  (bind _returnvalue_4 (- _returnvalue _value_2))
  (bind _knockbackactorrefreshtime_const (Variables|Knockback|GetKnockbackActorRefreshTime%20Const))
  (bind _instigator (Variables|Actor|GetInstigator))
  (Networking|SwitchHasAuthority
    (:Authority
      (Utilities|IsValid Overlapping%20Actor
        (:Is%20Valid
          (if (and (!= self Overlapping%20Actor) (Collision|GetActorEnableCollision Overlapping%20Actor))
            (Utilities|FlowControl|Sequence
              (:then_0
                ;[0.4115 0.1846 0.01084] Allow hostile sawblades to destroy each other
                ; Check the overlapping sawblade against the instigator to see if they are hostile and destroy each other if they are
                (bind _asprj_athena_sawblade (Utilities|Casting|CastToPrj_Athena_Sawblade Overlapping%20Actor))
                (Utilities|IsValid _instigator
                  (:Is%20Valid
                    (if (Utilities|Enum|Equal(Enum) (Team|GetTeamAffiliation _instigator _asprj_athena_sawblade) "Hostile")
                      (Actor|DestroyActor _asprj_athena_sawblade)))))
              (:then_1
                ;[0.1506 0.6729 1] If this actor is NOT in our knockback times list OR its time has expired THEN attempt to knock it back.
                ; Also calculate TimeTilNextOverlapTrigger which is when we should next trigger
                (if _returnvalue_3
                  (if (not (>= _returnvalue_4 _knockbackactorrefreshtime_const))
                    ;;[1 0.8984 0] The knockback cooldown on this actor is still in progress, Set our overlap timer to trigger when the cooldown is finished.
                    (LocalVariables|SetTimeTilNextOverlapTrigger (- _knockbackactorrefreshtime_const _returnvalue_4)))
                  (else
                    _))
                ;;[0.2633 1 0.3058] The knockback cooldown doesn't exists or has exited, attempt a knockback and set our overlap timer to trigger once the new cooldown has finished
                (Variables|Cached|SetLastHitInfo (Collision|MakeHitResult false false 0.0 0.0 "0, 0, 0" "0, 0, 0" (* (Transformation|GetActorForwardVector) -1.0) "0,0,1" 0 Overlapping%20Actor))
                (Variables|Cached|SetOtherActorHit Overlapping%20Actor)
                (CallFunction|DetermineProjectileHitBehavior)
                (LocalVariables|SetTimeTilNextOverlapTrigger _knockbackactorrefreshtime_const))
              (:then_2
                ;[0.06198 1 0.9702] If there is no timer for this overlapping actor OR the timer has expired
                ; THEN create a new overlap timer.
                (if _returnvalue_2
                  (if (<= (Utilities|Time|GetTimerRemainingTimebyHandle _value) 0)
                    _))
                (bind _returnvalue_5 (Utilities|Time|SetTimerbyEvent (EventDispatchers|CreateEvent :!event KnockbackOverlappingActorTimerExpired) (LocalVariables|GetTimeTilNextOverlapTrigger)))
                (Utilities|Map|Add _knockbacktimerhandlesbyoverlappingvictimactormap Overlapping%20Actor _returnvalue_5)))))))))

(fn EndOverlap
  (params (Other%20Actor :type Actor%20Object%20Reference))
  (bind _knockbacktimerhandlesbyoverlappingvictimactormap (Variables|Knockback|GetKnockbackTimerHandlesByOverlappingVictimActorMap))
  (bind (_value _returnvalue) (Utilities|Map|Find _knockbacktimerhandlesbyoverlappingvictimactormap Other%20Actor))
  (Networking|SwitchHasAuthority
    (:Authority
      (if _returnvalue
        (Utilities|Time|ClearandInvalidateTimerbyHandle _value)
        (Utilities|Map|Remove _knockbacktimerhandlesbyoverlappingvictimactormap Other%20Actor)
        (Utilities|Map|Remove _knockbacktimerhandlesbyoverlappingvictimactormap)))))

(fn GetMaterialSurfaceIndex
  (params (HitData :type Hit%20Result%20Structure))
  (bind (_bblockinghit _binitialoverlap _time _distance _location _impactpoint _normal _impactnormal _physmat) (Collision|BreakHitResult HitData))
  (Utilities|IsValid _physmat
    (:Is%20Valid
      ;[1 0.185 0.662] Return a default value if the physical surface is invalid, otherwise return the correct index
      (return (Utilities|Select :Index (Variables|/Script/PhysicsCore.PhysicalMaterial|GetSurfaceType _physmat) :SurfaceType_Default 0 :SurfaceType1 1 :SurfaceType2 2 :SurfaceType3 3 :SurfaceType4 4 :SurfaceType5 5 :SurfaceType6 6 :SurfaceType7 7 :SurfaceType8 8 :SurfaceType9 9 :SurfaceType10 10 :SurfaceType11 11 :SurfaceType12 12 :SurfaceType13 13 :SurfaceType14 14 :SurfaceType15 15 :SurfaceType16 16 :SurfaceType17 17 :SurfaceType18 18 :SurfaceType19 19 :SurfaceType20 20 :SurfaceType21 21 :SurfaceType22 22 :SurfaceType23 23 :SurfaceType24 24 :SurfaceType25 25 :SurfaceType26 26 :SurfaceType27 0 :SurfaceType28 0 :SurfaceType29 0 :SurfaceType30 0 :SurfaceType31 0 :SurfaceType32 0 :SurfaceType33 0 :SurfaceType34 0) (Variables|/Script/PhysicsCore.PhysicalMaterial|GetSurfaceType _physmat)))
    (:Is%20Not%20Valid
      (return 0 "SurfaceType_Default"))))

(fn KnockbackHitActor
  (params (HitResult :type Hit%20Result%20Structure) (HitActor :type Actor%20Object%20Reference))
  (bind _hf_knockbackimpulsemagnitude (Variables|Knockback|GetHF_KnockbackImpulseMagnitude))
  (bind _returnvalue (ScalableFloat|GetValueatLevel _hf_knockbackimpulsemagnitude))
  (bind _knockbackimpulsezoffset_const (Variables|Knockback|GetKnockbackImpulseZOffset%20Const))
  (bind _returnvalue_2 (Math|Vector|MakeVector 0.0 0.0 _knockbackimpulsezoffset_const))
  (bind (_bblockinghit _binitialoverlap _time _distance _location _impactpoint _normal) (Collision|BreakHitResult HitResult))
  (bind _capsulecomponent (Variables|Collision|GetCapsuleComponent))
  (bind _returnvalue_3 (Transformation|GetWorldLocation _capsulecomponent))
  (bind _returnvalue_4 (* _normal _returnvalue))
  (bind _returnvalue_5 (Fortnite|Sync|Time|GetServerWorldTimeSeconds))
  (bind _returnvalue_6 (+ _returnvalue_4 _returnvalue_2))
  (Utilities|FlowControl|Sequence
    (:then_0
      ;[0.7969 0.8221 1] Handle knocking back the actor
      (bind (_asfort_pawn _asfort_vehicle_interface _building_actor) (Utilities|CasttoKnockbackSupportingActor HitActor)
        (:Fort%20Pawn
          ;;[0.02173 0.02163 0.349] Handle knocking back Fort Pawns
          (if (!= (Variables|Actor|GetInstigator) _asfort_pawn)
            (DamagingAndHit|ApplyImpactDamage HitResult _asfort_pawn)
            (else
              _))
          (DamagingAndHit|QueueLaunchCharacter _asfort_pawn (* _returnvalue_6 (Variables|Knockback|GetKnockbackPawnImpulseMultiplier%20Const))) ; Do NOT Launch the character immediately because we may be within the character's movement scope and changing the character's velocity at this point could cause a crash.
        )
        (:Fort%20Vehicle%20Interface
          ;;[0 0.3953 1] Handle knocking back vehicles
          (if (not (ScalableFloat|GetValueasBool (Variables|DamagingAndHit|GetHF_EnableVehicleFriendlyFire)))
            (Utilities|IsValid (Vehicle|GetDriver _asfort_vehicle_interface)
              (:Is%20Valid
                (if (Utilities|Enum|Equal(Enum) (Team|GetTeamAffiliation (Vehicle|GetDriver _asfort_vehicle_interface)) "Hostile")
                  _))
              (:Is%20Not%20Valid
                _)))
          (DamagingAndHit|ApplyImpactDamage HitResult (Utilities|ToObject(Interface) _asfort_vehicle_interface))
          (Audio|BroadcastAbilityImpactatLocation 0 (Utilities|ToObject(Interface) _asfort_vehicle_interface) _returnvalue_6 777.0 "") ; This launches vehicles (named weird for anti-cheat reasons)\nPosition = velocity to launch it at\nRadius must equal 777\nGates knows more about it
        )
        (:Knockbackable%20Building%20Actor
          ;;[0.4445 0.04843 1] Handle knocking back buildings.
          ;; Do NOT knockback a building if it is the building we are currently stopped at.
          (if (!= _building_actor (Variables|Cached|GetBuildingStoppedAt))
            (CallFunction|DealDamagetoBuilding _building_actor (Collision|BreakHitResult HitResult))))
        (:Unsupported
          ;;[1 0.4844 0.4844] The actor we are trying to knockback is NOT supported, abort function.
          (return))))
    (:then_1
      ;[0.4132 0.02567 1] Maybe add VFX/SFX for this impact
      (if (>= (- _returnvalue_5 (Variables|Knockback|GetKnockbackLastCosmeticsTime)) (Variables|Knockback|GetKnockbackActorRefreshTime%20Const))
        (Variables|Knockback|SetKnockbackLastCosmeticsTime _returnvalue_5)
        (Effects|Components|ParticleSystem|SpawnFXAtLocation (Variables|VFX|GetNS_ImpulseKnockback) _returnvalue_3)
        (Audio|PlaySoundatLocation (Variables|SFX|GetSFX_ImpulseKnockback) _returnvalue_3)))))

(fn OnHitSetInitialVelocities
  (bind _desired_orthogonal_direction (Variables|Cached|GetDesired%20Orthogonal%20Direction))
  (Networking|SwitchHasAuthority
    (:Authority
      (Variables|StateTrackers|SetHasEverHitSomething true)
      ;[0.2958 0.447 0.6042] Set up velocities, the first time we hit something
      (if (not (Math|Vector|VectorIsZero _desired_orthogonal_direction))
        (Variables|Cached|SetGoalVelocity%20XY (Math|Conversions|ToVector2D(Vector) (* _desired_orthogonal_direction (Variables|Cached|GetDesired%20Orthogonal%20Speed))))
        (Variables|Cached|SetCurrentDesiredVelocity%20XY (Math|Conversions|ToVector2D(Vector) (Variables|/Script/Engine.MovementComponent|GetVelocity (Variables|Movement|GetProjectileMovementComponent))))
        (Traversal|ConstraintoVerticalPlane)))))

(fn OnRep_GoalVelocityXY
  (Variables|StateTrackers|SetAtMaxSpeed false) ; Since the goal velocity changed reset AtMaxSpeed so that the current speed will update.
)

(fn OnRep_PlaneNormalConstraintReplicated
  (bind _projectilemovementcomponent (Variables|Movement|GetProjectileMovementComponent))
  (Components|Movement|Planar|SetPlaneConstraintNormal _projectilemovementcomponent (Variables|Traversal|GetPlaneNormalConstraint%20Replicated))
  (Components|Movement|Planar|SetPlaneConstraintEnabled _projectilemovementcomponent true))

(fn OnRep_PlaneOriginConstraintReplicated
  (bind _projectilemovementcomponent (Variables|Movement|GetProjectileMovementComponent))
  (Components|Movement|Planar|SetPlaneConstraintOrigin _projectilemovementcomponent (Variables|Traversal|GetPlaneOriginConstraint%20Replicated))
  (Components|Movement|Planar|SetPlaneConstraintEnabled _projectilemovementcomponent true))

(fn OnRep_Stopped
  (bind _projectilemovementcomponent (Variables|Movement|GetProjectileMovementComponent))
  (Utilities|FlowControl|Sequence
    (:then_0
      (if (Variables|StateTrackers|GetStopped)
        ;[1 0.5589 0.00038] On Projectile Stopped
        (Variables|Cached|SetPreStoppedVelocity (Variables|/Script/Engine.MovementComponent|GetVelocity _projectilemovementcomponent))
        (Projectile|SetVelocity _projectilemovementcomponent (Math|Vector|MakeVector 0.0))
        (Variables|/Script/Engine.ProjectileMovementComponent|SetProjectileGravityScale 0.0 _projectilemovementcomponent)
        (bind _asfort_projectile_move_comp_slider (Utilities|Casting|CastToFortProjectileMoveComp_Slider _projectilemovementcomponent)
          (:then
            (Class|FortProjectileMoveCompSlider|SetSuspensionEnabled _asfort_projectile_move_comp_slider false false)
            (Class|FortProjectileMoveCompSlider|ZeroOutSuspensionTension _asfort_projectile_move_comp_slider))
          (:CastFailed
            _))
        (CallFunction|EndStuckProjectileCheck)
        (Utilities|Time|PauseTimerbyHandle (Variables|VFX|GetTH_TraversalAirborneTimer)) ; pause the VFX timer when damaging an object \nbecause no new hits should be registered
        (else
          ;[0 1 0.1746] On Projectile un-stopped.
          ;;[0.4132 0.02567 1] VFX: Remove Gameplay Cue (impact/damage)
          (Networking|SwitchHasAuthority
            (:Authority
              _)
            (:Remote
              (Components|Activation|Deactivate (Variables|Default|GetVFX_Niagra_Damage))))
          (Utilities|Time|UnpauseTimerbyHandle (Variables|VFX|GetTH_TraversalAirborneTimer)) ; unpause traversal timer since new hit events can occur
          (Variables|StateTrackers|SetAtMaxSpeed false)
          (Variables|Cached|SetPreStoppedVelocity "0, 0, 0")
          (Variables|Cached|SetCurrentDesiredVelocity%20XY (* (Variables|Cached|GetGoalVelocity%20XY) (Math|Vector2D|MakeVector2D 0.15 0.15)))
          (CallFunction|StartStuckProjectileCheck)
          (Utilities|Time|ClearandInvalidateTimerbyHandle (Variables|StateTrackers|GetDealDamageTimer))
          (CallFunction|ResetDamagingVFX)
          (bind _asfort_projectile_move_comp_slider_2 (Utilities|Casting|CastToFortProjectileMoveComp_Slider (Variables|Movement|GetProjectileMovementComponent)))
          (Class|FortProjectileMoveCompSlider|SetSuspensionEnabled _asfort_projectile_move_comp_slider_2 true false))))))

(fn PlayHitMarkerAudio
  (params (Instigator :type Pawn%20Object%20Reference) (Target :type Object%20Reference))
  (bind _asfort_pawn (Utilities|Casting|CastToFortPawn Target))
  (if (Pawn|IsLocallyControlled Instigator)
    (if (FortPawn|IsDead _asfort_pawn)
      (bind _returnvalue (CallFunction|GetOwnerWeapon))
      (bind _asb_sawblade_gun_athena (Utilities|Casting|CastToB_SawbladeGun_Athena _returnvalue))
      (Class|BSawbladeGunAthena|Play2DHitNotification _asb_sawblade_gun_athena))))

(fn QueueLaunchCharacter
  (params (PawnToLaunch :type Fort%20Pawn%20Object%20Reference) (LaunchVelocity :type Vector))
  (Utilities|Array|Add (Variables|Knockback|GetPawnLaunchData) (Utilities|Struct|MakeSTRUCTSawbladeLaunchCharacter PawnToLaunch LaunchVelocity)))

(fn ReduceLifeSpantoDieAfterWarning
  (bind _lifespanendwarnthreshold_const (Variables|VFX|GetLifeSpanEndWarnThreshold_const))
  ;[0.401 0.2071 0.04981] ReduceLifeSpanToDieAfterWarning
  ; Reduces the life spawn of the item when it is not already in the warn of immenent death state, then call warning communication
  (if (> (Actor|GetLifeSpan) _lifespanendwarnthreshold_const)
    (Actor|SetLifeSpan _lifespanendwarnthreshold_const)
    (CallFunction|LifeSpanNearEndWarning)
    (CallFunction|EndStuckProjectileCheck)))

(fn RemoveStaleKnockbackActors
  (bind _knockbacktimesbyvictimactormap (Variables|Knockback|GetKnockbackTimesByVictimActorMap))
  (Utilities|FlowControl|Sequence
    (:then_0
      ;[1 0.5108 0.01025] Remove all actors that have become null
      (Utilities|Map|Remove _knockbacktimesbyvictimactormap))
    (:then_1
      ;[0.1872 1 0.1985] Compile a list of stale actors.
      (bind _output_get (LocalVariables|SetStaleCutOffTime (- (Fortnite|Sync|Time|GetServerWorldTimeSeconds) (Variables|Knockback|GetKnockbackActorRefreshTime%20Const))))
      (bind _keys (Utilities|Map|Keys _knockbacktimesbyvictimactormap))
      (bind _values (Utilities|Map|Values _knockbacktimesbyvictimactormap))
      (for (_array_element _array_index) _values
        ;;[1 0.9628 0.1707] If a Knockback actor's time is before the cut off time add it to the list of actors to remove.
        (if (< _array_element _output_get)
          (Utilities|Array|Add (LocalVariables|GetActorsToRemove) (Utilities|Array|Get(acopy) _keys _array_index)))))
    (:then_2
      ;[0.4531 0.2985 0.05192] Remove Stale Actors from Map
      (for _array_element_2 (LocalVariables|GetActorsToRemove)
        (Utilities|Map|Remove _knockbacktimesbyvictimactormap _array_element_2)))))

(fn SetDamageAudioSurfaceType
  (params (AudioComponent :type Audio%20Component%20Object%20Reference) (HitResult :type Hit%20Result%20Structure))
  (bind (_bblockinghit _binitialoverlap _time _distance _location _impactpoint _normal _impactnormal _physmat) (Collision|BreakHitResult HitResult))
  (bind _surfacetype (Variables|/Script/PhysicsCore.PhysicalMaterial|GetSurfaceType _physmat))
  (bind _damageloopaudiomap (Variables|SFX|GetDamageLoopAudioMap))
  (bind (_value _returnvalue) (Utilities|Map|Find _damageloopaudiomap _surfacetype))
  (if (Variables|SFX|GetbDamageAudioFaded)
    (Variables|SFX|SetbDamageAudioFaded false)
    (Audio|Components|Audio|FadeIn AudioComponent 0.2 1.0 0.0 "Linear" :!class /Script/Engine.AudioComponent))
  (if _returnvalue
    (if (not (== _value (Variables|SFX|GetAudioSurfaceIndex)))
      (Variables|SFX|SetAudioSurfaceIndex _value)
      (Audio|Parameter|SetIntegerParameter AudioComponent "PhysicalSurface" _value :!class /Script/AudioExtensions.AudioParameterControllerInterface)
      (return _value))
    (elif (== (Variables|SFX|GetAudioSurfaceIndex) 0)
      _
      (else
        (Variables|SFX|SetAudioSurfaceIndex 0)
        (Audio|Parameter|SetIntegerParameter AudioComponent "PhysicalSurface" 0 :!class /Script/AudioExtensions.AudioParameterControllerInterface)
        (return 0)))))

(fn TickVelocity
  (bind _currentdesiredvelocity_xy (Variables|Cached|GetCurrentDesiredVelocity%20XY))
  (bind _goalvelocity_xy (Variables|Cached|GetGoalVelocity%20XY))
  (bind _returnvalue (- _goalvelocity_xy _currentdesiredvelocity_xy))
  (bind _projectilemovementcomponent (Variables|Movement|GetProjectileMovementComponent))
  (bind (_currentdesiredvelocity_xy_x _currentdesiredvelocity_xy_y) (Math|Vector2D|BreakVector2D (Variables|GetCurrentDesiredVelocity%20XY_X)))
  (bind (_velocity_x _velocity_y _velocity_z) (Math|Vector|BreakVector (Variables|/Script/Engine.MovementComponent|GetVelocity_X _projectilemovementcomponent)))
  (if (and (Variables|StateTrackers|GetHasEverHitSomething) (not (Variables|StateTrackers|GetStopped)))
    (Utilities|FlowControl|Sequence
      (:then_0
        ;[0.1235 0.9502 1] Set the velocity and gravity based on whether or not blade is Stopped (cutting through something)
        (bind _output_get (LocalVariables|SetDesiredVelocity (Math|Vector|MakeVector _currentdesiredvelocity_xy_x _currentdesiredvelocity_xy_y _velocity_z)))
        (Projectile|SetVelocity _projectilemovementcomponent _output_get)
        (Variables|/Script/Engine.ProjectileMovementComponent|SetProjectileGravityScale 1.0 _projectilemovementcomponent)
        ;[0.1235 0.9502 1] Accelerate back up to max speed after having been stopped
        (if (not (Variables|StateTrackers|GetAtMaxSpeed))
          (Variables|Cached|SetCurrentDesiredVelocity%20XY (+ _currentdesiredvelocity_xy (* _returnvalue (Math|Vector2D|MakeVector2D (* (Utilities|Time|GetWorldDeltaSeconds) (Variables|Traversal|GetDesired%20Acceleration%20XY%20Const)) (* (Utilities|Time|GetWorldDeltaSeconds) (Variables|Traversal|GetDesired%20Acceleration%20XY%20Const))))))
          ;[0 0.03342 1] If the desired velocity XY is within the tolerance of the goal velocity XY.
          (if (< (Math|Vector2D|Vector2DLengthSquared _returnvalue) (Math|Float|Square (Variables|Traversal|GetGoal%20Velocity%20XY%20Tolerance%20Const)))
            (Variables|StateTrackers|SetAtMaxSpeed true)
            (Variables|Cached|SetCurrentDesiredVelocity%20XY (Variables|Cached|GetGoalVelocity%20XY))))))))

(fn ConstructionScript
  (|CallParentFunction))
