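# Sound cue for the sawblade gun projectile being destroyed.
# A Random node picks among four equally weighted metal impact waves taken from the projectile's physical-material switch sounds.
# The cue uses the sawblade projectile attenuation asset and plays under the ranged weapon sound class.
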SoundCue /SawbladeGun/Items/SawbladeGun/Sounds/Gun_Sawblade_Proj_Destroyed_Cue
root=Random_0
nodes [
 {id=Random_0, children [WavePlayer_0,WavePlayer_1,WavePlayer_2,WavePlayer_3], properties {weights [1,1,1,1]}}
 {id=WavePlayer_0, properties {soundWaveAssetPtr=&/Gun_Sawblade_Proj_PhysMatSwitch_Metal_04}}
 {id=WavePlayer_1, properties {soundWaveAssetPtr=&/Gun_Sawblade_Proj_PhysMatSwitch_Metal_01}}
 {id=WavePlayer_2, properties {soundWaveAssetPtr=&/Gun_Sawblade_Proj_PhysMatSwitch_Metal_02}}
 {id=WavePlayer_3, properties {soundWaveAssetPtr=&/Gun_Sawblade_Proj_PhysMatSwitch_Metal_03}}
]
properties {
 attenuationSettings=&/SawBladeGun_Proj_Att
 maxDistance=6400
 soundClassObject=&Fort_WeaponRanged
}
