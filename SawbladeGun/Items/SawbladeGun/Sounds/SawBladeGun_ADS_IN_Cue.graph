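# Sound cue for the SawbladeGun aim-down-sights-in foley, named for the moment the weapon is raised to aim.
# Its only node is a root Modulator at default settings with an empty input, so no wave is connected and the cue has nothing to play.
# It still carries the weapon foley sound class, the shared weapon reload attenuation and a reduced volume multiplier.
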SoundCue /SawbladeGun/Items/SawbladeGun/Sounds/SawBladeGun_ADS_IN_Cue
root=Modulator_0
nodes [{id=Modulator_0, children [null]}]
properties {
 attenuationSettings=&Fort_Weapon_Reload_Att
 maxDistance=2250
 soundClassObject=&Fort_WeaponFoley
 volumeMultiplier=0.376
}
