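# Sound cue for the sawblade gun's charge-up, routed to the third-person ranged weapon fire sound class.
# A Mixer layers one of four randomly picked charge-start one-shots over a looping charged sound at half input volume.
# An Enveloper fades that loop in over its first second, so it rises beneath the start sound.
# A dedicated charge attenuation asset handles spatial falloff, and overridden concurrency rejects new instances at the limit.
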SoundCue /SawbladeGun/Items/SawbladeGun/Sounds/SawBladeGun_ChargeUp_Cue
root=Mixer_1
nodes [
 {
  id=Enveloper_0
  children [WavePlayer_2]
  properties {pitchCurve {editorCurveData {keys [{time=0}]}}, volumeCurve {editorCurveData {keys [<at 0 0>,<at 0.504 0.3739>,<at 1 1>]}}}
 }
 {id=Mixer_1, children [Enveloper_0,Random_0], properties {inputVolume [0.5,1]}}
 {id=Random_0, children [WavePlayer_1,WavePlayer_0,WavePlayer_3,WavePlayer_4], properties {weights [1,1,1,1]}}
 {id=WavePlayer_0, properties {soundWaveAssetPtr=&/Gun_Sawblade_Charge_Start_02}}
 {id=WavePlayer_1, properties {soundWaveAssetPtr=&/Gun_Sawblade_Charge_Start_01}}
 {id=WavePlayer_2, properties {bLooping=true, soundWaveAssetPtr=&/Gun_Sawblade_Charged_Loop_01}}
 {id=WavePlayer_3, properties {soundWaveAssetPtr=&/Gun_Sawblade_Charge_Start_03}}
 {id=WavePlayer_4, properties {soundWaveAssetPtr=&/Gun_Sawblade_Charge_Start_04}}
]
properties {
 attenuationSettings=&/SawBladeGun_Charge_Att
 bOverrideConcurrency=true
 maxDistance=2100
 soundClassObject=&Fort_WeaponRanged_Fire_3P_PSM
 volumeMultiplier=0.6
 attenuationOverrides {bEnableReverbSend=false, distanceAlgorithm=Logarithmic, falloffDistance=1000, stereoSpread=100, attenuationShapeExtents {x=125}}
 concurrencyOverrides {resolutionRule=PreventNew}
}
