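# Sound cue for the sawblade gun reaching full charge, layering a charge-complete sound with a charged loop.
# A Mixer blends the one-shot with the looping wave, weighting the one-shot louder.
# An Enveloper shapes the loop's pitch and volume over time, fading it in under the one-shot, holding, then easing it lower.
# It uses the gun's charge attenuation, a third-person weapon-fire sound class and a shared concurrency rule.
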SoundCue /SawbladeGun/Items/SawbladeGun/Sounds/SawBladeGun_Charge_Full_Cue
root=Mixer_0
nodes [
 {
  id=Enveloper_0
  children [WavePlayer_2]
  properties {
   pitchCurve {editorCurveData {keys [{time=0}, <at 1 1.1>, {time=20}, <at 60 1.1>]}}
   volumeCurve {editorCurveData {keys [<at 0 0>,<at 0.3044 0.2419>,<at 0.6 1>,<at 3 1>,<at 8 0.6>]}}
  }
 }
 {id=Mixer_0, children [WavePlayer_0,Enveloper_0], properties {inputVolume [1.2,0.7]}}
 {id=WavePlayer_0, properties {soundWaveAssetPtr=&/Gun_Sawblade_Charge_Full_01}}
 {id=WavePlayer_2, properties {bLooping=true, soundWaveAssetPtr=&/Gun_Sawblade_Charged_Loop_02}}
]
properties {
 attenuationSettings=&/SawBladeGun_Charge_Att
 maxDistance=2100
 soundClassObject=&Fort_WeaponRanged_Fire_3P_PSM
 volumeMultiplier=0.7
 concurrencyOverrides {maxCount=2}
 concurrencySet [/Game/Sounds/Concurrency_Rules/Athena_Prop]
}
