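# First-person fire sound cue for the Sawblade Gun, layering a close gunshot with a blade cocking sound.
# A mixer combines a randomly chosen fire variation with a randomly chosen, quieter cock variation routed through a modulator, each drawn from four equally weighted waves.
# It uses first-person weapon attenuation, a ranged-fire sound class and shared local and global gunfire concurrency sets.
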SoundCue /SawbladeGun/Items/SawbladeGun/Sounds/SawBladeGun_Fire_1P_Cue
root=Mixer_0
nodes [
 {id=Mixer_0, children [Random_3,Modulator_0], properties {inputVolume [1,0.6]}}
 {id=Modulator_0, children [Random_0]}
 {id=Random_0, children [WavePlayer_0,WavePlayer_1,WavePlayer_2,WavePlayer_3], properties {weights [1,1,1,1]}}
 {id=Random_3, children [WavePlayer_5, WavePlayer_10, WavePlayer_11, WavePlayer_12], properties {weights [1,1,1,1]}}
 {id=WavePlayer_0, properties {soundWaveAssetPtr=&/Gun_Sawblade_Blade_Cock_1p_01}}
 {id=WavePlayer_1, properties {soundWaveAssetPtr=&/Gun_Sawblade_Blade_Cock_1p_02}}
 {id=WavePlayer_10, properties {soundWaveAssetPtr=&/Gun_Sawblade_Fire_Close_02}}
 {id=WavePlayer_11, properties {soundWaveAssetPtr=&/Gun_Sawblade_Fire_Close_03}}
 {id=WavePlayer_12, properties {soundWaveAssetPtr=&/Gun_Sawblade_Fire_Close_04}}
 {id=WavePlayer_2, properties {soundWaveAssetPtr=&/Gun_Sawblade_Blade_Cock_1p_03}}
 {id=WavePlayer_3, properties {soundWaveAssetPtr=&/Gun_Sawblade_Blade_Cock_1p_04}}
 {id=WavePlayer_5, properties {soundWaveAssetPtr=&/Gun_Sawblade_Fire_Close_01}}
]
properties {
 attenuationSettings=&Fort_Weapon_Projectile_FirstPerson_Athena_Att
 maxDistance=2097152
 soundClassObject=&Fort_WeaponRanged_Fire_1P_PSM
 volumeMultiplier=0.9
 attenuationOverrides {falloffDistance=5000, attenuationShapeExtents {x=1000}}
 concurrencyOverrides {bLimitToOwner=true, maxCount=3, resolutionRule=StopOldest}
 concurrencySet [&Weapon_Gunfire_Local_Conc, &Weapon_Gunfire_Global_Conc]
}
