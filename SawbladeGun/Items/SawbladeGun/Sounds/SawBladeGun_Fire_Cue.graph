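# Sound cue for the Sawblade Gun's fire sound, layered by distance into close, medium and distant shots.
# Each layer picks one of four random wave variations, then passes through a modulator and its own projectile attenuation.
# The close layer uses the third-person fire sound class, while medium and distant share a distant-fire class and get a louder final mix.
# Local and global gunfire concurrency sets govern overlapping shots.
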SoundCue /SawbladeGun/Items/SawbladeGun/Sounds/SawBladeGun_Fire_Cue
root=Mixer_1
nodes [
 {
  id=Attenuation_10
  children [Modulator_14]
  properties {attenuationSettings=&Fort_Weapon_Projectile_Medium_Mono_Athena_Att}
 }
 {
  id=Attenuation_11
  children [Modulator_13]
  properties {attenuationSettings=&Fort_Weapon_Projectile_Distant_Mono_Athena_Att}
 }
 {id=Attenuation_9, children [Modulator_15], properties {attenuationSettings=&Fort_Weapon_Projectile_Close_Athena_Att}}
 {id=Mixer_1, children [SoundClass_7,SoundClass_6], properties {inputVolume [1,1.5]}}
 {id=Mixer_2, children [Attenuation_10, Attenuation_11], properties {inputVolume [1,1]}}
 {id=Modulator_13, children [Random_5], properties {pitchMax=1.1}}
 {id=Modulator_14, children [Random_4], properties {pitchMax=1.1, volumeMax=1, volumeMin=1}}
 {id=Modulator_15, children [Random_3]}
 {id=Random_3, children [WavePlayer_5, WavePlayer_10, WavePlayer_11, WavePlayer_12], properties {weights [1,1,1,1]}}
 {id=Random_4, children [WavePlayer_13, WavePlayer_14, WavePlayer_15, WavePlayer_16], properties {weights [1,1,1,1]}}
 {id=Random_5, children [WavePlayer_17, WavePlayer_18, WavePlayer_19, WavePlayer_20], properties {weights [1,1,1,1]}}
 {id=SoundClass_6, children [Mixer_2], properties {soundClassOverride=&Fort_WeaponRanged_Fire_Distant}}
 {
  id=SoundClass_7
  children [Attenuation_9]
  properties {soundClassOverride=&Fort_WeaponRanged_Fire_3P_PSM}
 }
 {id=WavePlayer_10, properties {soundWaveAssetPtr=&/Gun_Sawblade_Fire_Close_02}}
 {id=WavePlayer_11, properties {soundWaveAssetPtr=&/Gun_Sawblade_Fire_Close_03}}
 {id=WavePlayer_12, properties {soundWaveAssetPtr=&/Gun_Sawblade_Fire_Close_04}}
 {id=WavePlayer_13, properties {soundWaveAssetPtr=&/Gun_Sawblade_Fire_Med_01}}
 {id=WavePlayer_14, properties {soundWaveAssetPtr=&/Gun_Sawblade_Fire_Med_02}}
 {id=WavePlayer_15, properties {soundWaveAssetPtr=&/Gun_Sawblade_Fire_Med_03}}
 {id=WavePlayer_16, properties {soundWaveAssetPtr=&/Gun_Sawblade_Fire_Med_04}}
 {id=WavePlayer_17, properties {soundWaveAssetPtr=&/Gun_Sawblade_Fire_Dist_01}}
 {id=WavePlayer_18, properties {soundWaveAssetPtr=&/Gun_Sawblade_Fire_Dist_02}}
 {id=WavePlayer_19, properties {soundWaveAssetPtr=&/Gun_Sawblade_Fire_Dist_03}}
 {id=WavePlayer_20, properties {soundWaveAssetPtr=&/Gun_Sawblade_Fire_Dist_04}}
 {id=WavePlayer_5, properties {soundWaveAssetPtr=&/Gun_Sawblade_Fire_Close_01}}
]
properties {
 maxDistance=27500
 volumeMultiplier=0.9
 attenuationOverrides {falloffDistance=5000, attenuationShapeExtents {x=1000}}
 concurrencyOverrides {bLimitToOwner=true, maxCount=3, resolutionRule=StopOldest}
 concurrencySet [&Weapon_Gunfire_Local_Conc, &Weapon_Gunfire_Global_Conc]
}
