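# Sound cue for the sawblade gun firing its last shot and running empty.
# A single empty-last-shot wave plays through a modulator after a short fixed delay.
# It uses the third-person ranged weapon fire sound class, a close projectile attenuation asset, and local and global gunfire concurrency sets.
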SoundCue /SawbladeGun/Items/SawbladeGun/Sounds/SawBladeGun_Fire_LastShot_Empty_Cue
root=Delay_0
nodes [
 {id=Delay_0, children [Modulator_1], properties {delayMax=0.1, delayMin=0.1}}
 {id=Modulator_1, children [WavePlayer_4]}
 {id=WavePlayer_4, properties {soundWaveAssetPtr=&/Gun_Sawblade_Empty_Last_Shot_01}}
]
properties {
 attenuationSettings=&Fort_Weapon_Projectile_Close_Athena_Att
 maxDistance=4950
 soundClassObject=&Fort_WeaponRanged_Fire_3P_PSM
 volumeMultiplier=0.7
 attenuationOverrides {falloffDistance=5000, attenuationShapeExtents {x=1000}}
 concurrencyOverrides {bLimitToOwner=true, maxCount=3, resolutionRule=StopOldest}
 concurrencySet [&Weapon_Gunfire_Local_Conc, &Weapon_Gunfire_Global_Conc]
}
