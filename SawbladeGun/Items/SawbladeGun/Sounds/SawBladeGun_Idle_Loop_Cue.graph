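# Sound cue for the sawblade gun's idle loop, routed as weapon foley with reload attenuation.
# A looping wave player feeds an enveloper whose volume curve fades in quickly, then gradually decays over about half a minute so long idles grow quieter.
# A root modulator randomizes pitch slightly downward.
# The cue plays at reduced overall volume, overrides concurrency and has its debug flag on.
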SoundCue /SawbladeGun/Items/SawbladeGun/Sounds/SawBladeGun_Idle_Loop_Cue
root=Modulator_0
nodes [
 {
  id=Enveloper_0
  children [WavePlayer_0]
  properties {
   pitchCurve {editorCurveData {keys [{time=0}]}}
   volumeCurve {editorCurveData {keys [<at 0 0>,<at 1 1>,<at 4.0978 0.9899>,<at 12.0948 0.7>,<at 30 0.35>]}}
  }
 }
 {id=Modulator_0, children [Enveloper_0], properties {pitchMax=1, pitchMin=0.85}}
 {id=WavePlayer_0, properties {bLooping=true, soundWaveAssetPtr=&/Gun_Sawblade_Idle_Loop_01}}
]
properties {
 attenuationSettings=&Fort_Weapon_Reload_Att
 bDebug=true
 bOverrideConcurrency=true
 maxDistance=2250
 soundClassObject=&Fort_WeaponFoley
 volumeMultiplier=0.251
}
