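# Looping sound cue for Sawblade Gun blade grinding, named for projectile melee damage, that varies with the surface hit.
# A PhysicalSurface switch selects generic, stone, dirt, wood, metal, flesh or water loops, with random choice between two waves for wood, metal and flesh.
# A modulator slightly varies pitch before the root mixer.
# The cue uses melee attenuation, a distance-delay source effect chain and per-owner concurrency that stops the oldest instance.
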SoundCue /SawbladeGun/Items/SawbladeGun/Sounds/SawBladeGun_Proj_Damage_Melee_Loop_Cue
root=Mixer_1
nodes [
 {id=Mixer_1, children [Modulator_0], properties {inputVolume [1]}}
 {id=Mixer_10, children [Random_2], properties {inputVolume [1]}}
 {id=Mixer_11, children [WavePlayer_2], properties {inputVolume [1]}}
 {id=Mixer_3, children [WavePlayer_13], properties {inputVolume [1]}}
 {id=Mixer_4, children [WavePlayer_7], properties {inputVolume [1]}}
 {id=Mixer_5, children [WavePlayer_8], properties {inputVolume [1]}}
 {id=Mixer_7, children [Random_0], properties {inputVolume [1]}}
 {id=Mixer_8, children [Random_1], properties {inputVolume [1]}}
 {id=Modulator_0, children [Switch_1], properties {pitchMax=1.1, pitchMin=0.9}}
 {id=Random_0, children [WavePlayer_4,WavePlayer_5], properties {weights [1,1]}}
 {id=Random_1, children [WavePlayer_10, WavePlayer_6], properties {weights [1,1]}}
 {id=Random_2, children [WavePlayer_11, WavePlayer_12], properties {weights [1,1]}}
 {id=Switch_1, children [Mixer_3,Mixer_3,Mixer_4,Mixer_5,Mixer_7,Mixer_8,Mixer_10,Mixer_11], properties {intParameterName=PhysicalSurface}}
 {
  id=WavePlayer_10
  properties {bLooping=true, soundWaveAssetPtr=&/Gun_Sawblade_Proj_Blade_Grind_Metal_Loop_01}
 }
 {
  id=WavePlayer_11
  properties {bLooping=true, soundWaveAssetPtr=&/Gun_Sawblade_Proj_Blade_Grind_Flesh_Loop_02}
 }
 {
  id=WavePlayer_12
  properties {bLooping=true, soundWaveAssetPtr=&/Gun_Sawblade_Proj_Blade_Grind_Flesh_Loop_01}
 }
 {id=WavePlayer_13, properties {bLooping=true, soundWaveAssetPtr=&/Gun_Sawblade_Proj_Blade_Grind_Loop_01}}
 {
  id=WavePlayer_2
  properties {bLooping=true, soundWaveAssetPtr=&/Gun_Sawblade_Proj_Blade_Grind_Water_Loop_01}
 }
 {
  id=WavePlayer_4
  properties {bLooping=true, soundWaveAssetPtr=&/Gun_Sawblade_Proj_Blade_Grind_Wood_Loop_01}
 }
 {
  id=WavePlayer_5
  properties {bLooping=true, soundWaveAssetPtr=&/Gun_Sawblade_Proj_Blade_Grind_Wood_Loop_02}
 }
 {
  id=WavePlayer_6
  properties {bLooping=true, soundWaveAssetPtr=&/Gun_Sawblade_Proj_Blade_Grind_Metal_Loop_02}
 }
 {
  id=WavePlayer_7
  properties {bLooping=true, soundWaveAssetPtr=&/Gun_Sawblade_Proj_Blade_Grind_Stone_Loop_01}
 }
 {
  id=WavePlayer_8
  properties {bLooping=true, soundWaveAssetPtr=&/Gun_Sawblade_Proj_Blade_Grind_Dirt_Loop_02}
 }
]
properties {
 attenuationSettings=&/SawBladeGun_Melee_Att
 bOverrideConcurrency=true
 maxDistance=4000
 soundClassObject=&Fort_WeaponRanged
 sourceEffectChain=&/SawBlade_Gun_Delay_Distance_PC
 volumeMultiplier=1.25
 attenuationOverrides {bAttenuate=false, bSpatialize=false}
 concurrencyOverrides {bLimitToOwner=true, maxCount=3, resolutionRule=StopOldest}
}
