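# Sound cue for the fly-by sound of the sawblade gun's projectile, classed as a ranged weapon sound.
# An equally weighted random node picks one of five fly-by recordings, which a modulator pitches randomly between slightly lowered and normal.
# It references a dedicated fly-by attenuation asset and overrides concurrency with a rule that blocks new instances once the limit is reached.
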SoundCue /SawbladeGun/Items/SawbladeGun/Sounds/SawBladeGun_Proj_FlyBy_Cue
root=Modulator_0
nodes [
 {id=Modulator_0, children [Random_0], properties {pitchMax=1, pitchMin=0.8}}
 {id=Random_0, children [WavePlayer_1,WavePlayer_2,WavePlayer_3,WavePlayer_4,WavePlayer_5], properties {weights [1,1,1,1,1]}}
 {id=WavePlayer_1, properties {soundWaveAssetPtr=&/Gun_Sawblade_Proj_FlyBy_01}}
 {id=WavePlayer_2, properties {soundWaveAssetPtr=&/Gun_Sawblade_Proj_FlyBy_02}}
 {id=WavePlayer_3, properties {soundWaveAssetPtr=&/Gun_Sawblade_Proj_FlyBy_03}}
 {id=WavePlayer_4, properties {soundWaveAssetPtr=&/Gun_Sawblade_Proj_FlyBy_04}}
 {id=WavePlayer_5, properties {soundWaveAssetPtr=&/Gun_Sawblade_Proj_FlyBy_05}}
]
properties {
 attenuationSettings=&/SawBladeGun_Proj_FlyBy_Att
 bOverrideConcurrency=true
 maxDistance=4500
 soundClassObject=&Fort_WeaponRanged
 volumeMultiplier=1
 attenuationOverrides {distanceAlgorithm=NaturalSound, falloffDistance=2500, attenuationShapeExtents {x=500}}
 concurrencyOverrides {resolutionRule=PreventNew}
}
