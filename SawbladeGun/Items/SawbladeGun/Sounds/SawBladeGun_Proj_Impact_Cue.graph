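# Sound cue for the sawblade gun's projectile impacts, varying the sound by the surface hit.
# A Switch on the PhysicalSurface parameter picks among Random nodes, each holding four equally weighted generic, stone, wood or metal waves, then feeds a root Modulator.
# It uses the projectile attenuation asset and the Fort_Impacts sound class, with overridden per-owner concurrency that stops the oldest sound.
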SoundCue /SawbladeGun/Items/SawbladeGun/Sounds/SawBladeGun_Proj_Impact_Cue
root=Modulator_0
nodes [
 {id=Modulator_0, children [Switch_1]}
 {id=Random_0, children [WavePlayer_1,WavePlayer_2,WavePlayer_3,WavePlayer_4], properties {weights [1,1,1,1]}}
 {id=Random_1, children [WavePlayer_5, WavePlayer_29, WavePlayer_30, WavePlayer_31], properties {weights [1,1,1,1]}}
 {id=Random_2, children [WavePlayer_32, WavePlayer_33, WavePlayer_34, WavePlayer_35], properties {weights [1,1,1,1]}}
 {id=Random_3, children [WavePlayer_36, WavePlayer_37, WavePlayer_38, WavePlayer_39], properties {weights [1,1,1,1]}}
 {id=Random_4, children [WavePlayer_40, WavePlayer_41, WavePlayer_42, WavePlayer_43], properties {weights [1,1,1,1]}}
 {id=Random_5, children [WavePlayer_21, WavePlayer_22, WavePlayer_23, WavePlayer_24], properties {weights [1,1,1,1]}}
 {id=Random_6, children [WavePlayer_25, WavePlayer_26, WavePlayer_27, WavePlayer_28], properties {weights [1,1,1,1]}}
 {
  id=Switch_1
  children [Random_0,Random_0,Random_1,Random_2,Random_3,Random_4,Random_5,Random_6]
  properties {intParameterName=PhysicalSurface}
 }
 {id=WavePlayer_1, properties {soundWaveAssetPtr=&/Gun_Sawblade_Proj_Impt_01}}
 {id=WavePlayer_2, properties {soundWaveAssetPtr=&/Gun_Sawblade_Proj_Impt_02}}
 {id=WavePlayer_21, properties {soundWaveAssetPtr=&Gun_Sawblade_Proj_Impt_01}}
 {id=WavePlayer_22, properties {soundWaveAssetPtr=&Gun_Sawblade_Proj_Impt_02}}
 {id=WavePlayer_23, properties {soundWaveAssetPtr=&/Gun_Sawblade_Proj_Impt_03}}
 {id=WavePlayer_24, properties {soundWaveAssetPtr=&/Gun_Sawblade_Proj_Impt_04}}
 {id=WavePlayer_25, properties {soundWaveAssetPtr=&Gun_Sawblade_Proj_Impt_01}}
 {id=WavePlayer_26, properties {soundWaveAssetPtr=&Gun_Sawblade_Proj_Impt_02}}
 {id=WavePlayer_27, properties {soundWaveAssetPtr=&Gun_Sawblade_Proj_Impt_03}}
 {id=WavePlayer_28, properties {soundWaveAssetPtr=&Gun_Sawblade_Proj_Impt_04}}
 {id=WavePlayer_29, properties {soundWaveAssetPtr=&/Gun_Sawblade_Proj_Impt_Stone_02}}
 {id=WavePlayer_3, properties {soundWaveAssetPtr=&Gun_Sawblade_Proj_Impt_03}}
 {id=WavePlayer_30, properties {soundWaveAssetPtr=&/Gun_Sawblade_Proj_Impt_Stone_03}}
 {id=WavePlayer_31, properties {soundWaveAssetPtr=&/Gun_Sawblade_Proj_Impt_Stone_04}}
 {id=WavePlayer_32, properties {soundWaveAssetPtr=&/Gun_Sawblade_Proj_Impt_Stone_01}}
 {id=WavePlayer_33, properties {soundWaveAssetPtr=&Gun_Sawblade_Proj_Impt_Stone_02}}
 {id=WavePlayer_34, properties {soundWaveAssetPtr=&Gun_Sawblade_Proj_Impt_Stone_03}}
 {id=WavePlayer_35, properties {soundWaveAssetPtr=&Gun_Sawblade_Proj_Impt_Stone_04}}
 {id=WavePlayer_36, properties {soundWaveAssetPtr=&/Gun_Sawblade_Proj_Impt_Wood_01}}
 {id=WavePlayer_37, properties {soundWaveAssetPtr=&/Gun_Sawblade_Proj_Impt_Wood_02}}
 {id=WavePlayer_38, properties {soundWaveAssetPtr=&/Gun_Sawblade_Proj_Impt_Wood_03}}
 {id=WavePlayer_39, properties {soundWaveAssetPtr=&/Gun_Sawblade_Proj_Impt_Wood_04}}
 {id=WavePlayer_4, properties {soundWaveAssetPtr=&Gun_Sawblade_Proj_Impt_04}}
 {id=WavePlayer_40, properties {soundWaveAssetPtr=&/Gun_Sawblade_Proj_Impt_Metal_01}}
 {id=WavePlayer_41, properties {soundWaveAssetPtr=&/Gun_Sawblade_Proj_Impt_Metal_02}}
 {id=WavePlayer_42, properties {soundWaveAssetPtr=&/Gun_Sawblade_Proj_Impt_Metal_03}}
 {id=WavePlayer_43, properties {soundWaveAssetPtr=&/Gun_Sawblade_Proj_Impt_Metal_04}}
 {id=WavePlayer_5, properties {soundWaveAssetPtr=&Gun_Sawblade_Proj_Impt_Stone_01}}
]
properties {
 attenuationSettings=&/SawBladeGun_Proj_Att
 bOverrideConcurrency=true
 maxDistance=6400
 soundClassObject=&Fort_Impacts
 volumeMultiplier=2.8224
 attenuationOverrides {bAttenuate=false, bSpatialize=false}
 concurrencyOverrides {bLimitToOwner=true, maxCount=3, resolutionRule=StopOldest}
}
