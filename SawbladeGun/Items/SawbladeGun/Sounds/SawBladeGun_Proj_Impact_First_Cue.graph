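# Sound cue for the Sawblade Gun projectile's first-hit impact, routed to the Fort_Impacts sound class.
# A Random node picks among four equally weighted FirstHit wave variants and feeds a Modulator that applies slight random pitch variation.
# It takes spatial behavior from the shared projectile attenuation asset, boosts volume, and overrides concurrency to cap simultaneous plays per owner, stopping the oldest.
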SoundCue /SawbladeGun/Items/SawbladeGun/Sounds/SawBladeGun_Proj_Impact_First_Cue
root=Modulator_0
nodes [
 {id=Modulator_0, children [Random_0], properties {pitchMax=1, pitchMin=0.9}}
 {id=Random_0, children [WavePlayer_1,WavePlayer_2,WavePlayer_3,WavePlayer_4], properties {weights [1,1,1,1]}}
 {id=WavePlayer_1, properties {soundWaveAssetPtr=&/Gun_Sawblade_Proj_Impt_FirstHit_03}}
 {id=WavePlayer_2, properties {soundWaveAssetPtr=&/Gun_Sawblade_Proj_Impt_FirstHit_04}}
 {id=WavePlayer_3, properties {soundWaveAssetPtr=&/Gun_Sawblade_Proj_Impt_FirstHit_01}}
 {id=WavePlayer_4, properties {soundWaveAssetPtr=&/Gun_Sawblade_Proj_Impt_FirstHit_02}}
]
properties {
 attenuationSettings=&/SawBladeGun_Proj_Att
 bOverrideConcurrency=true
 maxDistance=6400
 soundClassObject=&Fort_Impacts
 volumeMultiplier=2.4192
 attenuationOverrides {bAttenuate=false, bSpatialize=false}
 concurrencyOverrides {bLimitToOwner=true, maxCount=3, resolutionRule=StopOldest}
}
