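# Sound cue for sawblade gun melee projectile impacts, choosing audio by the surface hit.
# A PhysicalSurface switch selects random pools of generic, stone, wood, metal or water-splash waves, or a mix of projectile and generic impacts.
# A root modulator randomly lowers pitch for variety.
# It uses melee attenuation and the impacts sound class, with a per-owner concurrency cap that stops the oldest sound.
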SoundCue /SawbladeGun/Items/SawbladeGun/Sounds/SawBladeGun_Proj_Impact_Melee_Cue
root=Modulator_0
nodes [
 {id=Mixer_1, children [Random_5,Random_7], properties {inputVolume [1,0.7]}}
 {id=Modulator_0, children [Switch_1], properties {pitchMax=1, pitchMin=0.85}}
 {id=Random_0, children [WavePlayer_1,WavePlayer_2,WavePlayer_3,WavePlayer_4], properties {weights [1,1,1,1]}}
 {id=Random_1, children [WavePlayer_9, WavePlayer_10, WavePlayer_11, WavePlayer_12], properties {weights [1,1,1,1]}}
 {id=Random_2, children [WavePlayer_13, WavePlayer_14, WavePlayer_15, WavePlayer_16], properties {weights [1,1,1,1]}}
 {id=Random_3, children [WavePlayer_17, WavePlayer_18, WavePlayer_19, WavePlayer_20], properties {weights [1,1,1,1]}}
 {id=Random_4, children [WavePlayer_44, WavePlayer_45, WavePlayer_46, WavePlayer_47], properties {weights [1,1,1,1]}}
 {id=Random_5, children [WavePlayer_21, WavePlayer_22, WavePlayer_23, WavePlayer_24], properties {weights [1,1,1,1]}}
 {id=Random_6, children [WavePlayer_52, WavePlayer_53, WavePlayer_54, WavePlayer_55], properties {weights [1,1,1,1]}}
 {id=Random_7, children [WavePlayer_48, WavePlayer_49, WavePlayer_50, WavePlayer_51], properties {weights [1,1,1,1]}}
 {
  id=Switch_1
  children [Random_0,Random_0,Random_1,Random_2,Random_3,Random_4,Mixer_1,Random_6]
  properties {intParameterName=PhysicalSurface}
 }
 {id=WavePlayer_1, properties {soundWaveAssetPtr=&/Gun_Sawblade_Proj_Impt_Melee_Gen_01}}
 {id=WavePlayer_10, properties {soundWaveAssetPtr=&/Gun_Sawblade_Proj_Impt_Melee_Stone_01}}
 {id=WavePlayer_11, properties {soundWaveAssetPtr=&/Gun_Sawblade_Proj_Impt_Melee_Stone_02}}
 {id=WavePlayer_12, properties {soundWaveAssetPtr=&/Gun_Sawblade_Proj_Impt_Melee_Stone_03}}
 {id=WavePlayer_13, properties {soundWaveAssetPtr=&/Gun_Sawblade_Proj_Impt_Melee_Stone_04}}
 {id=WavePlayer_14, properties {soundWaveAssetPtr=&Gun_Sawblade_Proj_Impt_Melee_Stone_01}}
 {id=WavePlayer_15, properties {soundWaveAssetPtr=&Gun_Sawblade_Proj_Impt_Melee_Stone_02}}
 {id=WavePlayer_16, properties {soundWaveAssetPtr=&Gun_Sawblade_Proj_Impt_Melee_Stone_03}}
 {id=WavePlayer_17, properties {soundWaveAssetPtr=&/Gun_Sawblade_Proj_Impt_Melee_Wood_04}}
 {id=WavePlayer_18, properties {soundWaveAssetPtr=&/Gun_Sawblade_Proj_Impt_Melee_Wood_01}}
 {id=WavePlayer_19, properties {soundWaveAssetPtr=&/Gun_Sawblade_Proj_Impt_Melee_Wood_02}}
 {id=WavePlayer_2, properties {soundWaveAssetPtr=&/Gun_Sawblade_Proj_Impt_Melee_Gen_02}}
 {id=WavePlayer_20, properties {soundWaveAssetPtr=&/Gun_Sawblade_Proj_Impt_Melee_Wood_03}}
 {id=WavePlayer_21, properties {soundWaveAssetPtr=&/Gun_Sawblade_Proj_Impt_01}}
 {id=WavePlayer_22, properties {soundWaveAssetPtr=&/Gun_Sawblade_Proj_Impt_02}}
 {id=WavePlayer_23, properties {soundWaveAssetPtr=&/Gun_Sawblade_Proj_Impt_03}}
 {id=WavePlayer_24, properties {soundWaveAssetPtr=&/Gun_Sawblade_Proj_Impt_04}}
 {id=WavePlayer_3, properties {soundWaveAssetPtr=&/Gun_Sawblade_Proj_Impt_Melee_Gen_03}}
 {id=WavePlayer_4, properties {soundWaveAssetPtr=&/Gun_Sawblade_Proj_Impt_Melee_Gen_04}}
 {id=WavePlayer_44, properties {soundWaveAssetPtr=&/Gun_Sawblade_Proj_Impt_Melee_Metal_01}}
 {id=WavePlayer_45, properties {soundWaveAssetPtr=&/Gun_Sawblade_Proj_Impt_Melee_Metal_02}}
 {id=WavePlayer_46, properties {soundWaveAssetPtr=&/Gun_Sawblade_Proj_Impt_Melee_Metal_03}}
 {id=WavePlayer_47, properties {soundWaveAssetPtr=&/Gun_Sawblade_Proj_Impt_Melee_Metal_04}}
 {id=WavePlayer_48, properties {soundWaveAssetPtr=&Gun_Sawblade_Proj_Impt_Melee_Gen_01}}
 {id=WavePlayer_49, properties {soundWaveAssetPtr=&Gun_Sawblade_Proj_Impt_Melee_Gen_02}}
 {id=WavePlayer_50, properties {soundWaveAssetPtr=&Gun_Sawblade_Proj_Impt_Melee_Gen_03}}
 {id=WavePlayer_51, properties {soundWaveAssetPtr=&Gun_Sawblade_Proj_Impt_Melee_Gen_04}}
 {
  id=WavePlayer_52
  properties {soundWaveAssetPtr=/FNGameplayCues/Sounds/Physics/Objects/WaterSplash/MediumObject/Phys_Medium_Object_Splash_Heavy_04}
 }
 {
  id=WavePlayer_53
  properties {soundWaveAssetPtr=/FNGameplayCues/Sounds/Physics/Objects/WaterSplash/MediumObject/Phys_Medium_Object_Splash_Heavy_03}
 }
 {
  id=WavePlayer_54
  properties {soundWaveAssetPtr=/FNGameplayCues/Sounds/Physics/Objects/WaterSplash/MediumObject/Phys_Medium_Object_Splash_Heavy_02}
 }
 {
  id=WavePlayer_55
  properties {soundWaveAssetPtr=/FNGameplayCues/Sounds/Physics/Objects/WaterSplash/MediumObject/Phys_Medium_Object_Splash_Heavy_01}
 }
 {id=WavePlayer_9, properties {soundWaveAssetPtr=&Gun_Sawblade_Proj_Impt_Melee_Stone_04}}
]
properties {
 attenuationSettings=&/SawBladeGun_Melee_Att
 bOverrideConcurrency=true
 maxDistance=4000
 soundClassObject=&Fort_Impacts
 volumeMultiplier=2.4192
 attenuationOverrides {bAttenuate=false, bSpatialize=false}
 concurrencyOverrides {bLimitToOwner=true, maxCount=3, resolutionRule=StopOldest, retriggerTime=1}
}
