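# Sound cue for the looping blade spin audio on Sawblade Gun projectiles, classed as a ranged weapon sound.
# A looping wave player feeds a Modulator at the root, which applies pitch and volume variation at its usual settings.
# Playback uses projectile attenuation settings and a distance-delay source effect chain.
# A concurrency override caps instances per owner, stopping the oldest when the limit is exceeded.
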SoundCue /SawbladeGun/Items/SawbladeGun/Sounds/SawBladeGun_Proj_Loop_Cue
root=Modulator_1
nodes [
 {id=Modulator_1, children [WavePlayer_0]}
 {id=WavePlayer_0, properties {bLooping=true, soundWaveAssetPtr=&/Gun_Sawblade_Proj_Blade_Spin_Loop_01}}
]
properties {
 attenuationSettings=&/SawBladeGun_Proj_Att
 bOverrideConcurrency=true
 maxDistance=6400
 soundClassObject=&Fort_WeaponRanged
 sourceEffectChain=&/SawBlade_Gun_Delay_Distance_PC
 volumeMultiplier=1
 attenuationOverrides {distanceAlgorithm=NaturalSound}
 concurrencyOverrides {bLimitToOwner=true, maxCount=3, resolutionRule=StopOldest}
}
