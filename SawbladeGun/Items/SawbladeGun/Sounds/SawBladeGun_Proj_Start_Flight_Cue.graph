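# Sound cue for the sawblade gun projectile's start-of-flight sound, classed as a ranged weapon sound.
# A random node picks one of five equally weighted wave variations, which passes through a modulator that fixes a lowered pitch and a short constant delay before playback.
# The cue plays at reduced volume with its own attenuation asset, and its concurrency override prevents new instances once the limit is reached.
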SoundCue /SawbladeGun/Items/SawbladeGun/Sounds/SawBladeGun_Proj_Start_Flight_Cue
root=Delay_0
nodes [
 {id=Delay_0, children [Modulator_0], properties {delayMax=0.15, delayMin=0.15}}
 {id=Modulator_0, children [Random_0], properties {pitchMax=0.7, pitchMin=0.7}}
 {id=Random_0, children [WavePlayer_1,WavePlayer_2,WavePlayer_3,WavePlayer_4,WavePlayer_5], properties {weights [1,1,1,1,1]}}
 {id=WavePlayer_1, properties {soundWaveAssetPtr=&/Gun_Sawblade_Proj_Start_Flight_01}}
 {id=WavePlayer_2, properties {soundWaveAssetPtr=&/Gun_Sawblade_Proj_Start_Flight_02}}
 {id=WavePlayer_3, properties {soundWaveAssetPtr=&/Gun_Sawblade_Proj_Start_Flight_03}}
 {id=WavePlayer_4, properties {soundWaveAssetPtr=&/Gun_Sawblade_Proj_Start_Flight_04}}
 {id=WavePlayer_5, properties {soundWaveAssetPtr=&/Gun_Sawblade_Proj_Start_Flight_05}}
]
properties {
 attenuationSettings=&/SawBladeGun_Proj_Start_Att
 bOverrideConcurrency=true
 maxDistance=7000
 soundClassObject=&Fort_WeaponRanged
 volumeMultiplier=0.5
 attenuationOverrides {distanceAlgorithm=NaturalSound}
 concurrencyOverrides {resolutionRule=PreventNew}
}
