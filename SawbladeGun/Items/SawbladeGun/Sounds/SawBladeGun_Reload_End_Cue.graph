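# Sound cue for the end of the Sawblade Gun reload.
# A single wave player feeds the output directly, reusing an equip sound wave (Gun_Sawblade_Equip_01) with no randomization or modulation nodes.
# It uses the shared weapon reload attenuation and weapon foley sound class at slightly reduced volume.
# Concurrency override and the debug flag are both switched on.
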SoundCue /SawbladeGun/Items/SawbladeGun/Sounds/SawBladeGun_Reload_End_Cue
root=WavePlayer_1
nodes [{id=WavePlayer_1, properties {soundWaveAssetPtr=&/Gun_Sawblade_Equip_01}}]
properties {
 attenuationSettings=&Fort_Weapon_Reload_Att
 bDebug=true
 bOverrideConcurrency=true
 maxDistance=2250
 soundClassObject=&Fort_WeaponFoley
 volumeMultiplier=0.752
}
