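# Sound cue for the start of the Sawblade Gun reload.
# A single wave player plays one reload-start sound, with no randomization or modulation nodes.
# It routes to the weapon foley sound class and uses a weapon reload attenuation asset.
# A concurrency override allows one instance at a time, stopping the farthest before refusing new ones.
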SoundCue /SawbladeGun/Items/SawbladeGun/Sounds/SawBladeGun_Reload_Start_Cue
root=WavePlayer_0
nodes [{id=WavePlayer_0, properties {soundWaveAssetPtr=&/Gun_Sawblade_Reload_Start_01}}]
properties {
 attenuationSettings=&Fort_Weapon_Reload_Att
 bOverrideConcurrency=true
 maxDistance=2250
 soundClassObject=&Fort_WeaponFoley
 volumeMultiplier=0.752
 attenuationOverrides {falloffDistance=5000, attenuationShapeExtents {x=1000}}
 concurrencyOverrides {maxCount=1, resolutionRule=StopFarthestThenPreventNew}
}
