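# Sound cue that plays the sawblade gun's engine-off sound when the weapon is unequipped.
# A random node picks among four equally weighted engine shutdown wave variants, feeding a modulator that varies pitch for extra variety.
# It routes through the weapon foley sound class with reload attenuation at low volume.
# Overridden concurrency caps simultaneous instances, blocks new plays past the limit and sets a retrigger time, and the debug flag is left on.
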SoundCue /SawbladeGun/Items/SawbladeGun/Sounds/SawBladeGun_Unequip_Cue
root=Modulator_0
nodes [
 {id=Modulator_0, children [Random_0], properties {pitchMin=0.8}}
 {id=Random_0, children [WavePlayer_3,WavePlayer_4,WavePlayer_5,WavePlayer_6], properties {weights [1,1,1,1]}}
 {id=WavePlayer_3, properties {soundWaveAssetPtr=&/Gun_Sawblade_Unequip_Engine_Off_01}}
 {id=WavePlayer_4, properties {soundWaveAssetPtr=&/Gun_Sawblade_Unequip_Engine_Off_04}}
 {id=WavePlayer_5, properties {soundWaveAssetPtr=&/Gun_Sawblade_Unequip_Engine_Off_03}}
 {id=WavePlayer_6, properties {soundWaveAssetPtr=&/Gun_Sawblade_Unequip_Engine_Off_02}}
]
properties {
 attenuationSettings=&Fort_Weapon_Reload_Att
 bDebug=true
 bOverrideConcurrency=true
 maxDistance=2250
 soundClassObject=&Fort_WeaponFoley
 volumeMultiplier=0.175
 concurrencyOverrides {maxCount=3, resolutionRule=PreventNew, retriggerTime=0.5}
}
