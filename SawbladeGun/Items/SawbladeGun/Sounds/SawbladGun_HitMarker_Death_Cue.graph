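# Sound cue for the Sawblade Gun death hit marker, a notification sound that by its naming marks a lethal sawblade impact.
# A single wave player feeds the death impact notify wave straight to the output at slightly reduced volume, with no randomization or modulation.
# It uses a 2D hit-notification instigator sound class and overrides concurrency to one voice, stopping the oldest instance when retriggered.
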SoundCue /SawbladeGun/Items/SawbladeGun/Sounds/SawbladGun_HitMarker_Death_Cue
root=WavePlayer_3
nodes [{id=WavePlayer_3, properties {soundWaveAssetPtr=&/Gun_Sawblade_Impact_Death_Notify_01}}]
properties {
 bOverrideConcurrency=true
 maxDistance=2097152
 soundClassObject=&Fort_2DHitNotifications_Instigate_PSM
 volumeMultiplier=0.7
 concurrencyOverrides {maxCount=1, resolutionRule=StopOldest}
}
